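# Additive material for Niagara mesh particles, filed under the fluorescent lights effects of the Papaya event.
# A Fresnel term with a constant base reflect fraction feeds the Emissive input of the shared MF_FX_EmissiveAdjustments function, giving a view-angle-dependent intensity.
# The graph shows no textures or parameters of its own, and no explicit link from the function to a material output.
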Material /PapayaGameplay/Athena/Effects/Particles/Events/Papaya/FluorescentLights/AdditiveMaterial
graph {
 properties {bUsedWithNiagaraMeshParticles=true, blendMode=BLEND_Additive, pixelDepthOffsetMode=PDOM_Legacy, usedShadingModels=Default Lit}
 nodes [MaterialFunctionCall(Emissive=Fresnel(BaseReflectFractionIn=Constant() {r=0.125})) {materialFunction=&MF_FX_EmissiveAdjustments}]
}
