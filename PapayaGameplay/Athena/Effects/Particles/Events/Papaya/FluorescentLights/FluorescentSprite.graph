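# Additive unlit Niagara sprite material that draws camera-facing glare for fluorescent light tubes.
# Vertex offset math widens sprites along view X by random width, pulls them toward camera and shrinks them by camera distance to particle and dynamic-parameter endpoints.
# Opacity combines squared edge falloff, depth fade and a fade when viewed along the tube axis.
# Aspect-corrected screen-space lens dirt modulates glare; low-quality switches zero offset and opacity.
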Material /PapayaGameplay/Athena/Effects/Particles/Events/Papaya/FluorescentLights/FluorescentSprite
graph {
 properties {
  bUsedWithNiagaraSprites=true
  blendMode=BLEND_Additive
  pixelDepthOffsetMode=PDOM_Legacy
  shadingModel=MSM_Unlit
  twoSided=true
  usedShadingModels=Unlit
 }
 nodes [
  MaterialFunctionCall()
  n1 = Constant() {r=1}
  n2 = MaterialFunctionCall(Float2=MaterialFunctionCall() {materialFunction=&ScreenResolution}.Visible Resolution) {materialFunction=&BreakOutFloat2Components}
  n3 = TextureObject() {texture=/Engine/MapTemplates/T_ScreenDirt02_w}
  n4 = TextureProperty(Tex=@n3) {property=TMTM_TexelSize}
  n5 = MaterialFunctionCall(Float2=@n4) {materialFunction=&BreakOutFloat2Components}
  n6 = Divide(B=Multiply(A=Divide(A=@n2.G, B=@n2.G), B=Divide(A=@n5.R, B=@n5.R) {desc=ratio}))
  n7 = Constant() {r=1}
  n8 = ScalarParameter() {defaultValue=0.3381, parameterName=vfda}
  n9 = MaterialFunctionCall(
   UVs=ScreenPosition().ViewportUV
   Texture Scale=Multiply(
    A=AppendVector(A=@n1, B=@n6)
    B=MaterialFunctionCall(R=@n7, G=@n8) {materialFunction=&AppendMany}.RG
   )
  ) {materialFunction=&ScaleUVsByCenter}
  n10 = MaterialFunctionCall!(
   Base=Multiply(B=ScalarParameter() {defaultValue=0.2, parameterName=GlareBrightness})
   Add=Multiply(A=TextureSample(UVs=@n9.UVs, Tex=@n3).RGB, B=ScalarParameter() {defaultValue=168.5531, parameterName=screen dirt amount})
  ) {materialFunction=/Engine/Functions/Engine_MaterialFunctions02/Math/MultiplyAdd}
  MaterialFunctionCall(Emissive=@n10.Result) {materialFunction=&MF_FX_EmissiveAdjustments}
  MaterialFunctionCall(Texture Dimensions=@n4) {materialFunction=&ScreenAlignedUVs}
  n11 = LinearInterpolate(
   A=ScalarParameter() {defaultValue=57.2929, parameterName=Width}
   B=ScalarParameter() {defaultValue=424.0386, parameterName=Max WidthWidth}
   Alpha=ParticleRandom()
  )
  n12 = CameraPositionWS()
  n13 = DynamicParameter() {defaultValue=<rgba 0 0 0 1>, paramNames [TopPx, TopPy, TopPz, Percussion Lookup]}
  n14 = MaterialFunctionCall(Vector 3=Subtract(
    A=@n12
    B=MaterialFunctionCall(R=@n13.TopPx, G=@n13.TopPx, B=@n13.TopPx) {materialFunction=&AppendMany}.RGB
   )) {materialFunction=/Engine/Functions/Engine_MaterialFunctions02/Utility/VectorLength}
  n15 = DynamicParameter() {parameterIndex=1, defaultValue=<rgba 0 100 100 1>, paramNames [BottomPx, BottomPy, BottomPz, Percussion Lookup]}
  n16 = MaterialFunctionCall(Vector 3=Subtract(
    A=@n12
    B=MaterialFunctionCall(R=@n15.BottomPx, G=@n15.BottomPx, B=@n15.BottomPx) {materialFunction=&AppendMany}.RGB
   )) {materialFunction=&VectorLength}
  n17 = Subtract(A=@n12, B=ParticlePositionWS())
  n18 = Subtract(
   A=Min(
    A=Min(A=@n14.V3 Length, B=@n16.V3 Length)
    B=MaterialFunctionCall(Vector 3=@n17) {materialFunction=&VectorLength}.V3 Length
   )
   B=ScalarParameter() {defaultValue=128, parameterName=Reduce Flare Size Min Distance}
  )
  n19 = ScalarParameter() {defaultValue=2048, parameterName=Reduce Flare Size Falloff Distance}
  n20 = LinearInterpolate(Alpha=Saturate(None=Divide(A=@n18, B=@n19) {desc=Reduce Flare Max Distance}))
  n21 = ConstantBiasScale(None=MaterialFunctionCall(Float2=TextureCoordinate()) {materialFunction=&BreakOutFloat2Components}.R) {bias=-0.5, scale=2}
  n22 = Multiply(
   A=Transform(None=Constant3Vector() {constant=<rgba 1 0 0 0>}) {transformSourceType=TRANSFORMSOURCE_View}
   B=Multiply(A=@n21, B=Multiply(A=@n11, B=@n20))
  )
  n23 = WorldPosition()
  n24 = Add(
   A=@n22
   B=MaterialFunctionCall(Offset Amount=ScalarParameter() {defaultValue=100, parameterName=Camera Offset}, World Position=Add(A=@n22, B=@n23.XYZ)) {materialFunction=/Engine/Functions/Engine_MaterialFunctions02/WorldPositionOffset/CameraOffset}.Clamped Camera Offset
  )
  n25 = Constant()
  QualitySwitch(Default=@n24, Low=@n25, High=@n24, Medium=@n24)
  Reroute(None=@n10.Result)
  n26 = OneMinus(None=Abs(None=@n21))
  n27 = Multiply(A=@n26, B=@n26)
  n28 = ScalarParameter() {defaultValue=0.0391, parameterName=Opacity multiplier}
  n29 = ParticleDirection()
  n30 = ParticleSize() {desc="g = long side"}
  n31 = MaterialFunctionCall(Float2=TextureCoordinate() {desc=G length gradient}) {materialFunction=&BreakOutFloat2Components}
  n32 = Multiply(
   A=Multiply(
    A=MaterialFunctionCall(Float2=@n30) {materialFunction=&BreakOutFloat2Components}.G
    B=Subtract(A=@n31.G) {constB=0.5}
   )
   B=ParticleDirection()
  )
  n33 = ParticlePositionWS()
  n34 = MaterialFunctionCall(Vector=Subtract(A=CameraPositionWS(), B=Add(A=@n32, B=@n33) {desc=Vertical pivot points}) {desc=point toward camera}) {materialFunction=&SafeNormalize}
  n35 = OneMinus(None=Abs(None=DotProduct(A=@n29, B=@n34.Result)))
  n36 = Multiply(A=DepthFade(Opacity=Multiply(A=@n27, B=@n28)) {fadeDistanceDefault=200}, B=Clamp(None=Add(A=@n35, B=@n35)))
  Reroute(None=QualitySwitch(Default=@n36, Low=@n25, High=@n36, Medium=@n36))
  StaticBool() {value=true}
 ]
}
