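# Material function that reshapes a sphere mesh into a capsule spanning Top and Bottom positions, outputting a world position for WPO use and a world space normal.
# The normalized bottom-to-top axis becomes Z of a basis built with cross products, using safe normalizes with axis fallbacks.
# Local sphere positions and normals rotate into that basis; positions are then scaled and offset to either end by a top/bottom mask lerp.
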MaterialFunction /PapayaGameplay/Athena/Effects/Particles/Events/Papaya/FluorescentLights/MF_Papaya_SphereToCapsule
graph {nodes [
  n1 = FunctionInput() {inputName=Top Position, sortPriority=-2}
  n2 = FunctionInput() {inputName=Bottom Position, sortPriority=-1}
  n3 = MaterialFunctionCall(Vector=Subtract(A=@n1, B=@n2), Default=Constant3Vector() {constant=<rgba 0 0 1 0>}) {materialFunction=&SafeNormalize}
  n4 = MaterialFunctionCall(Vector=CrossProduct(A=@n3.Result, B=Constant3Vector() {constant=<rgba 1 0 0 0>}), Default=Constant3Vector() {constant=<rgba 0 1 0 0>}) {desc=+y, materialFunction=&SafeNormalize}
  n5 = MaterialFunctionCall(Vector=CrossProduct(A=@n4.Result, B=@n3.Result), Default=Constant3Vector() {constant=<rgba 1 0 0 0>}) {desc=x, materialFunction=&SafeNormalize}
  FunctionOutput(None=MaterialFunctionCall(VectorToTransform=FunctionInput() {inputName=Local Space Normals, sortPriority=2}, BasisX=@n5.Result, BasisY=@n4.Result, BasisZ=@n3.Result) {desc=World space normal, materialFunction=&Transform3x3Matrix}.Transformed Vector) {outputName=World space normal}
  n6 = MaterialFunctionCall(
   VectorToTransform=FunctionInput() {inputName=Scaled Sphere Local Position, sortPriority=1}
   BasisX=@n5.Result
   BasisY=@n4.Result
   BasisZ=@n3.Result
  ) {materialFunction=&Transform3x3Matrix}
  n7 = FunctionInput() {bUsePreviewValueAsDefault=true, inputName=Scale, inputType=FunctionInput_Scalar, previewValue=<1 0 0 1>}
  n8 = FunctionInput() {inputName=Top/Bottom BlackWhite Mask, inputType=FunctionInput_Scalar}
  FunctionOutput(None=Add(A=Multiply(A=@n6.Transformed Vector, B=@n7), B=LinearInterpolate(A=@n2, B=@n1, Alpha=@n8) {desc="Find "}) {desc="world position (subtract world position before using in WPO)"}) {outputName="world position (subtract world position before using in WPO)"}
 ]}
