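# Fluorescent light material for Niagara mesh particles in the Papaya event, shaded with LED panel textures.
# MF_Papaya_SphereToCapsule reshapes the mesh through world position offset and supplies normals, using dynamic-parameter top/bottom positions that replicate only X across all axes.
# Capsule scale and radius grow with camera distance.
# Pixel depth drives a temporally dithered near-camera dissolve, feeding opacity mask and, through HeightLerp, a huge pixel depth offset.
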Material /PapayaGameplay/Athena/Effects/Particles/Events/Papaya/FluorescentLights/M_Papaya_FluorescentLightMaterial
graph {
 properties {
  bCastRayTracedShadows=false
  bTangentSpaceNormal=false
  bUseMaterialAttributes=true
  bUsedWithNiagaraMeshParticles=true
  pixelDepthOffsetMode=PDOM_Legacy
  usedShadingModels=Default Lit
 }
 nodes [
  n1 = ParticleDirection()
  n2 = ParticleSize() {desc="g = long side"}
  n3 = ComponentMask(None=TextureCoordinate() {coordinateIndex=2}) {g=true}
  n4 = Subtract(A=AppendVector(A=TextureCoordinate() {coordinateIndex=1}, B=@n3)) {constB=0.5}
  n5 = Multiply(
   A=Multiply(
    A=MaterialFunctionCall(Float2=@n2) {materialFunction=&BreakOutFloat2Components}.G
    B=ComponentMask(None=@n4) {b=true}
   )
   B=ParticleDirection()
  )
  n6 = ParticlePositionWS()
  n7 = MaterialFunctionCall(Vector=Subtract(A=CameraPositionWS(), B=Add(A=@n5, B=@n6) {desc=Vertical pivot points}) {desc=point toward camera}) {materialFunction=&SafeNormalize}
  n8 = OneMinus(None=Abs(None=DotProduct(A=@n1, B=@n7.Result)))
  Clamp(None=Add(A=@n8, B=@n8))
  LinearInterpolate(
   A=VectorParameter() {defaultValue=<rgba 1 0 0 1>}.RGB
   B=VectorParameter() {parameterName=Param_1, defaultValue=<rgba 0 1 0 1>}.RGB
   Alpha=@n3
  )
  n9 = TextureSample() {samplerType=SAMPLERTYPE_Masks, texture=/Game/Environments/Papaya/Textures/T_Papaya_LEDPanel_06_S}
  n10 = PixelDepth()
  n11 = ScalarParameter() {defaultValue=120, parameterName=Dissolve start distance}
  n12 = Saturate(None=Divide(A=Subtract(A=@n10.R, B=@n11) {constB=30}, B=ScalarParameter() {defaultValue=57.4344, parameterName=Dissolve Distance}) {constB=90})
  n13 = MaterialFunctionCall(Alpha Threshold=@n12) {materialFunction=&DitherTemporalAA}
  n14 = MaterialFunctionCall(Transition Phase=@n12, Height Texture=@n13.Result) {materialFunction=&HeightLerp}
  n15 = DynamicParameter() {defaultValue=<rgba 0 0 0 1>, paramNames [TopPx, TopPy, TopPz, ""]}
  n16 = DynamicParameter() {parameterIndex=1, defaultValue=<rgba 0 100 100 1>, paramNames [BottomPx, BottomPy, BottomPz, ""]}
  n17 = ScalarParameter() {defaultValue=0.4857, parameterName=Top bottom Interpolation}
  n18 = MaterialFunctionCall(Vector 3=Subtract(A=ParticlePositionWS(), B=CameraPositionWS())) {materialFunction=/Engine/Functions/Engine_MaterialFunctions02/Utility/VectorLength}
  n19 = ScalarParameter() {defaultValue=1, parameterName=Scale Start Distance}
  n20 = Saturate(None=Divide(A=Subtract(A=@n18.V3 Length, B=@n19), B=ScalarParameter() {defaultValue=2048, parameterName=Scale Distance}))
  n21 = TextureSample() {samplerType=SAMPLERTYPE_Normal, texture=/Game/Environments/Papaya/Textures/T_Papaya_LEDPanel_06_N}
  n22 = ScalarParameter() {defaultValue=1.9452, desc=3.385929, parameterName=Radius Near}
  n23 = ScalarParameter() {defaultValue=3.3859, desc=3.385929, parameterName=Radius Far}
  n24 = MaterialFunctionCall!(
   Top Position=MaterialFunctionCall(R=@n15.TopPx, G=@n15.TopPx, B=@n15.TopPx) {materialFunction=&AppendMany}.RGB
   Bottom Position=MaterialFunctionCall(R=@n16.BottomPx, G=@n16.BottomPx, B=@n16.BottomPx) {materialFunction=&AppendMany}.RGB
   Top/Bottom BlackWhite Mask=MaterialFunctionCall(Alpha=@n3, Min=@n17, Max=Subtract(B=@n17)) {materialFunction=&SmoothStep}.Result
   Scale=LinearInterpolate(
    A=ScalarParameter() {defaultValue=1.9051, desc=6.4844, parameterName=Scale Near}
    B=ScalarParameter() {defaultValue=6.4844, desc=6.4844, parameterName=Scale Far}
    Alpha=@n20
   ) {desc="B = far away"}
   Scaled Sphere Local Position=Multiply(A=@n4, B=LinearInterpolate(A=@n22, B=@n23, Alpha=@n20))
   Local Space Normals=Multiply(
    A=VectorParameter() {parameterName=Flipped Normals, defaultValue=<rgba -1 1 -1 1>}.RGB
    B=Transform(None=@n21.RGB) {transformType=TRANSFORM_ParticleWorld}
   )
  ) {materialFunction=&/MF_Papaya_SphereToCapsule}
  MakeMaterialAttributes!(
   BaseColor=Multiply(A=TextureSample() {texture=/Game/Environments/Papaya/Textures/T_Papaya_LEDPanel_06_D}.RGB, B=@n9.A)
   Metallic=@n9.G
   Specular=@n9.G
   Roughness=@n9.G
   Anisotropy
   EmissiveColor=Multiply(A=Reroute(), B=OneMinus(None=@n9.A))
   Opacity=ScalarParameter() {defaultValue=0.75, parameterName=Opacity}
   OpacityMask=@n13.Result
   Normal=@n24.World space normal
   Tangent
   WorldPositionOffset=Subtract(A=@n24."world position (subtract world position before using in WPO)", B=WorldPosition().XYZ)
   SubsurfaceColor=VectorParameter() {parameterName=SSS Color, defaultValue=<rgba 0.0938 0.0938 0.0938 1>}.RGB
   ClearCoat
   ClearCoatRoughness
   AmbientOcclusion
   Refraction
   CustomizedUVs_0
   CustomizedUVs_1
   CustomizedUVs_2
   CustomizedUVs_3
   CustomizedUVs_4
   CustomizedUVs_5
   CustomizedUVs_6
   CustomizedUVs_7
   PixelDepthOffset=Multiply(A=OneMinus(None=@n14.Alpha)) {constB=19000000512}
   ShadingModel
   Displacement
  )
  MaterialFunctionCall(Float2=TextureCoordinate() {coordinateIndex=1, desc=G length gradient}) {materialFunction=&BreakOutFloat2Components}
  MaterialFunctionCall(
   Normal Vector=@n24.World space normal
   Invert Fresnel=StaticBool() {value=true}
   Power=ScalarParameter() {defaultValue=2, parameterName=Surface Fresnel Power}
  ) {materialFunction=&Fresnel_Function}
  MaterialFunctionCall()
  ScalarParameter() {defaultValue=5.5722, parameterName=Cap falloff pow}
  Subtract(A=@n3) {constB=0.5}
  n25 = TextureObject() {samplerType=SAMPLERTYPE_LinearColor, texture=/Game/Effects/Fort_Effects/Textures/Tile/LinerNoise64x64}
  TextureSample(
   UVs=MaterialFunctionCall!(TextureResolution=TextureProperty(Tex=@n25)) {materialFunction=/Engine/Functions/Engine_MaterialFunctions02/Texturing/ScreenAlignedPixelToPixelUVs}.Result
   Tex=@n25
  ) {samplerType=SAMPLERTYPE_LinearColor, texture=&LinerNoise64x64}
  TextureSample(UVs=Add(A=ParticleRandom(), B=TextureCoordinate())) {texture=&AtmosphericCloudNoise01}
  VertexNormalWS()
 ]
}
