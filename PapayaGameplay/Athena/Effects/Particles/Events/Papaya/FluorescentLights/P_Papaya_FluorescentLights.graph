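# Niagara system for Papaya event fluorescent lights, drawing particles as glowing tubes that pulse to jukebox percussion.
# Emitters represent arrangements, including a curly spawn line, skeletal-mesh dancers and exit-seeking Die; some, like Polyhedra, are disabled.
# Arrangements spawn on timed triggers, inheriting particle attributes and blending velocity from earlier ones.
# Each tube simulates top and bottom endpoints separately with springs and, in some arrangements, curl noise, feeding mesh, sprite and light renderers.
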NiagaraSystem /PapayaGameplay/Athena/Effects/Particles/Events/Papaya/FluorescentLights/P_Papaya_FluorescentLights
dependencies {
 usedDataInterfaces [
  &NiagaraDataInterfaceCurve
  &NiagaraDataInterfaceVectorCurve
  &NiagaraDataInterfaceArrayFloat2
  &NiagaraDataInterfaceParticleRead
 ]
 usedDynamicInputs []
 usedModules [
  &self:SystemSpawnScript.SystemScriptSource.SystemScriptGraph.NiagaraNodeAssignment_6.SetVariables_#1
  &self:SystemSpawnScript.SystemScriptSource.SystemScriptGraph.NiagaraNodeAssignment_1.SetVariables_#2
  &self:SystemSpawnScript.SystemScriptSource.SystemScriptGraph.NiagaraNodeAssignment_2.SetVariables_#3
  &self:SystemSpawnScript.SystemScriptSource.SystemScriptGraph.NiagaraNodeAssignment_0.SetVariables_#4
  &self:SystemSpawnScript.SystemScriptSource.SystemScriptGraph.NiagaraNodeAssignment_3.SetVariables_#5
  &self:SystemSpawnScript.SystemScriptSource.SystemScriptGraph.NiagaraNodeAssignment_5.SetVariables_#6
  &self:SystemSpawnScript.SystemScriptSource.SystemScriptGraph.NiagaraNodeAssignment_7.SetVariables_#7
  &self:SystemSpawnScript.SystemScriptSource.SystemScriptGraph.NiagaraNodeAssignment_4.SetVariables_#8
  &self:SystemSpawnScript.SystemScriptSource.SystemScriptGraph.NiagaraNodeAssignment_8.SetVariables_#9
  &self:SystemSpawnScript.SystemScriptSource.SystemScriptGraph.NiagaraNodeAssignment_9.SetVariables_#10
  &self:SystemSpawnScript.SystemScriptSource.SystemScriptGraph.NiagaraNodeAssignment_10.SetVariables_#11
  &SystemState
  &self:Find_All_Active_Emitter_Particle_Counts
  &self:FindTriggerMoments
 ]
}
emitters {
 Dancers {
  emitterSpawnScript [
   /SetVariables(Emitter.ThisEmittersArrangementsID:ENiagara_Papaya_FluorescentLightArrangements=::NewEnumerator1) {isSetParameters=true}
   /SetVariables(Emitter.DistanceBetweenSkeletalMeshSockets=-0.075, Emitter.OffsetBetweenSkeletalMeshCopies:Vector3f=<1025.1019 0 0>) {isSetParameters=true}
   /SetVariables(Emitter.ParticleBoneOptions=DataInterface() {propertyValues {GpuSyncMode=SyncCpuToGpu, MaxElements=0, FloatData [<0 1>, <2 3>, <3 4>, <5 6>, <6 7>, <8 9>, <10 11>, <12 13>, <14 15>, <9 12>, <11 14>]}}) {isSetParameters=true}
  ]
  emitterUpdateScript [
   &EmitterState()
   &self:GiveBirthToParticlesIfItsTheirTime() {module=GiveBirthToParticlesIfItsTheirTime}
  ]
  particleSpawnScript [
   /Niagara/Modules/Spawn/Initialization/InitializeParticle(
    Module.Write Color=false
    Module.Write Lifetime=true
    Module.Write Mass=false
    Module.Write Material Random=true
    Module.Write Position=false
    Module.Write Scale=false
    Module.Write SpriteRotation=false
    Module.Write SpriteSize=true
    Sprite Size=<200 50>
   )
   /SetVariables(Particles.Lag_Second_Point_Behind.TemporalLerp_Vector.Moving Average:Vector3f=$Particles.Position_Bottom) {isSetParameters=true}
   /SetVariables(Transient.TotalNumberOfCharacters:NiagaraInt32=/Niagara/DynamicInputs/Divide/Divide_Int(A:NiagaraInt32=$System.TotalNumberOfParticles, B:NiagaraInt32=11)) {isSetParameters=true}
   &self:Skeletal_AssignBoneIDs(
    New Vector 2D Array:NiagaraDataInterfaceArrayFloat2=$Emitter.ParticleBoneOptions
    SkeletalMesh:NiagaraDataInterfaceSkeletalMesh=$User.NewNiagaraDataInterfaceSkeletalMesh
   ) {module=Skeletal_AssignBoneIDs}
   &self:Find_FormationGroup_Centers(
    Cosine Offset Multiplier:Vector3f=<0 500 0>
    Sine Offset Multiplier:Vector3f=<700 0 0>
    Total Number of Groups:NiagaraInt32=$Transient.TotalNumberOfCharacters
    Total number of points per group=11
   ) {module=Find_FormationGroup_Centers}
   &self:Skeletal_FindSkeletalMeshTargetLocations(
    Group Center:Vector3f=$Particles.GroupCenter
    PullAmount:NiagaraFloat=$Emitter.DistanceBetweenSkeletalMeshSockets
    SkeletalMesh:NiagaraDataInterfaceSkeletalMesh=$User.NewNiagaraDataInterfaceSkeletalMesh
   ) {module=Skeletal_FindSkeletalMeshTargetLocations001}
   &self:ReadSenderOtherParticleAttributes() {module=ReadSenderOtherParticleAttributes}
   /SetVariables(Particles.PercussionLookupIndex:NiagaraFloat=$Output.NormalizedGroupID) {isSetParameters=true}
   /SetVariables(Particles.RandomEnergyBurstMultiplier:NiagaraFloat=&UniformRangedFloat(Minimum=0, Maximum=1)) {isSetParameters=true}
  ]
  particleUpdateScript [
   &self:MultiLerpPercussionSamples(PercussionLookupIndex:NiagaraFloat=$Particles.PercussionLookupIndex) {module=MultiLerpPercussionSamples}
   &self:PickRotationCenterVector(
    FrameCounter:NiagaraInt32=&UniformRangedInt(Minimum=0, Maximum=50000)
    GroupID:NiagaraInt32=&UniformRangedInt(Minimum=0, Maximum=50000)
    Rate Of Change:NiagaraFloat=&Multiply_Float(A=$Engine.DeltaTime, B=1)
   ) {module=PickRotationCenterVector}
   &self:Generate_Initial_Rotational_Force() {module=Generate_Initial_Rotational_Force}
   &ParticleState(Kill Particles When Lifetime Has Elapsed=false)
   &self:SeeIfParticlesNeedToDie(Number Of Frames To Linger:NiagaraInt32=$System.Number Of Frames To Linger) {module=SeeIfParticlesNeedToDie}
   &self:Skeletal_FindSkeletalMeshTargetLocations(
    Group Center:Vector3f=$Particles.GroupCenter
    PullAmount:NiagaraFloat=$Emitter.DistanceBetweenSkeletalMeshSockets
    SkeletalMesh:NiagaraDataInterfaceSkeletalMesh=$User.NewNiagaraDataInterfaceSkeletalMesh
   ) {module=Skeletal_FindSkeletalMeshTargetLocations}
   /Niagara/Modules/Update/Forces/CurlNoiseForce(
    Noise Strength=&FloatFromCurve(FloatCurve=DataInterface() {propertyValues {Curve {keys [<at 0 0>,<at 0.2403 1>,<at 1 0.25>]}}}, CurveIndex=$System.Age, Scale Curve=300)
    Pan Noise=true
    Pan Noise Field=<198.6738 3.8628 19.3902>
    Sample Position:Vector3f=$Particles.Position_Top
   )
   &CurlNoiseForce(
    Noise Strength=&FloatFromCurve(
     FloatCurve=DataInterface() {propertyValues {Curve {keys [<at 0 0>,<at 0.2403 1>,<at 0.9922 0.25>]}}}
     CurveIndex=$System.Age
     Scale Curve=300
    )
    Pan Noise=true
    Pan Noise Field=<198.6738 3.8628 19.3902>
    Sample Position:Vector3f=$Particles.Position_Bottom
   ) {module=CurlNoiseForce001}
   &CalculateAccurateVelocity(
    Current Position:Vector3f=$Particles.Top_TargetLocation
    Previous Position:Vector3f=$Particles.Previous.Top_TargetLocation
    Set Intrinsic Variables=false
   )
   &CalculateAccurateVelocity(
    Current Position:Vector3f=$Particles.Bottom_TargetLocation
    Previous Position:Vector3f=$Particles.Previous.Bottom_TargetLocation
    Set Intrinsic Variables=false
   ) {module=CalculateAccurateVelocity001}
   /Niagara/Modules/Update/Forces/SpringForce(
    " Particle Equilibrium Position":Vector3f=$Particles.Top_TargetLocation
    Attachment Point Velocity:Vector3f=$Output.CalculateAccurateVelocity.Velocity
    Dampening Coefficient=30
    Desired Minimum Separation Distance=0
    Particle Position:Vector3f=$Particles.Position_Top
    Particle Velocity:Vector3f=$Particles.Velocity_Top
    Spring Tightness=500
    Write to Intrinsic Parameters=false
   )
   &SpringForce(
    " Particle Equilibrium Position":Vector3f=$Particles.Bottom_TargetLocation
    Attachment Point Velocity:Vector3f=$Output.CalculateAccurateVelocity001.Velocity
    Dampening Coefficient=30
    Desired Minimum Separation Distance=0
    Particle Position:Vector3f=$Particles.Position_Bottom
    Particle Velocity:Vector3f=$Particles.Velocity_Bottom
    Spring Tightness=300
    Write to Intrinsic Parameters=false
   ) {module=SpringForce001}
   &self:Set_Top_Or_Bottom_Force(
    Bottom Force:Vector3f=&Add_Vector(A=$Output.SpringForce001.SpringForce, B=$Output.CurlNoiseForce001.CurlNoiseForce)
    Top Force:Vector3f=&Add_Vector(A=$Output.SpringForce.SpringForce, B=$Output.CurlNoiseForce.CurlNoiseForce)
   ) {module=Set_Top_Or_Bottom_Force}
   &self:-DUPLICATE_BELOW_THIS_MODULE-__Names() {module=-DUPLICATE_BELOW_THIS_MODULE-__Names}
   &self:BlendInInheritedVelocityAtSpawn(Clamped Normalized Age Curve:NiagaraDataInterfaceCurve=$System.Transfer to New Velocity Curve) {module=BlendInInheritedVelocityAtSpawn}
   &SolveForcesAndVelocity(
    Acceleration Limit:NiagaraFloat=$System.AccelerationLimit
    Clamp Velocity=true
    Force:Vector3f=$Transient.Top_PhysicsForce
    Limit Acceleration=true
    Position:Vector3f=$Particles.Position_Top
    PreviousVelocity:Vector3f=$Particles.Previous.Velocity_Top
    Speed Limit:NiagaraFloat=$System.MaxSpeed
    Velocity:Vector3f=$Particles.Velocity_Top
    Write to Intrinsic Properties=false
   )
   &SolveForcesAndVelocity(
    Acceleration Limit:NiagaraFloat=$System.AccelerationLimit
    Clamp Velocity=true
    Force:Vector3f=$Transient.Bottom_PhysicsForce
    Limit Acceleration=true
    Position:Vector3f=$Particles.Position_Bottom
    PreviousVelocity:Vector3f=$Particles.Previous.Velocity_Bottom
    Speed Limit:NiagaraFloat=$System.MaxSpeed
    Velocity:Vector3f=$Particles.Velocity_Bottom
    Write to Intrinsic Properties=false
   ) {module=SolveForcesAndVelocity001}
   &self:StoreOffTopAndBottomVariables(
    Bottom Position:NiagaraPosition=$Output.SolveForcesAndVelocity001.Position
    Bottom Velocity:Vector3f=$Output.SolveForcesAndVelocity001.Velocity
    Top Position:NiagaraPosition=$Output.SolveForcesAndVelocity.Position
    Top Velocity:Vector3f=$Output.SolveForcesAndVelocity.Velocity
   ) {module=StoreOffTopAndBottomVariables}
   &self:InitialTransferAnimation_(
    Include Initialization Rotation Animation=false
    RandomAxisOfRotation:Vector3f=$Particles.PickRotationCenterVector.RotationAxis
    Rotation Modulation:NiagaraFloat=$Particles.RandomEnergyBurstMultiplier
   ) {module=InitialTransferAnimation_}
   &self:FeedPositionsIntoShader(
    Bottom Position:Vector3f=$Particles.Position_Bottom
    Initial PercussionValue:NiagaraFloat=$Particles.Initial.PercussionValue
    PercussionValue:NiagaraFloat=$Particles.PercussionValue
    Top Position:Vector3f=$Particles.Position_Top
   ) {module=FeedPositionsIntoShader}
   /Niagara/Modules/Debug/SpriteBasedLine(Point A:Vector3f=$Particles.Position_Top, Point B:Vector3f=$Particles.Position_Bottom, SpriteWidth=0.01)
   &CalculateAccurateVelocity(
    Current Position:Vector3f=$Particles.Position_Top
    Previous Position:Vector3f=$Particles.Previous.Position_Top
    Set Intrinsic Variables=false
   ) {module=CalculateAccurateVelocity002}
   &CalculateAccurateVelocity(
    Current Position:Vector3f=$Particles.Position_Bottom
    Previous Position:Vector3f=$Particles.Previous.Position_Bottom
    Set Intrinsic Variables=false
   ) {module=CalculateAccurateVelocity003}
   /SetVariables(
    Particles.Position=$Particles.SpriteBasedLine.Position
    Particles.Velocity_Bottom:Vector3f=$Output.CalculateAccurateVelocity003.Velocity
    Particles.Velocity_Top:Vector3f=$Output.CalculateAccurateVelocity002.Velocity
   ) {isSetParameters=true}
   &self:CalculateLightSize(
    Color:LinearColor=&MakeLinearColorFromVectorAndFloat(
     "Vector (RGB)"=&MakeVectorFromLinearColorRGB(Color=&MultiplyLinearColor(
       Linear Color=&ColorFromCurve(ColorCurve=$System.BaseColorCurve, CurveIndex=&Frac_Float(Float=$NPC.Papaya_Jukebox_NPC.Base Color Parameter))
       Scale Factor=&MakeVector4_FromFloat(Float=&Lerp_Float(A=0.5, B=2, Alpha=$Particles.PercussionValue))
      ))
     "Float (Alpha)"=1
    )
    Sprite Based Line Size:Vector2f=$Particles.SpriteBasedLine.SpriteSize
   ) {module=CalculateLightSize}
  ]
  rendererClasses [
   &NiagaraMeshRendererProperties
   &NiagaraSpriteRendererProperties
   &NiagaraLightRendererProperties
  ]
  renderers [
   {rendererClass=&NiagaraMeshRendererProperties, rendererIndex=0}
   {rendererClass=&NiagaraSpriteRendererProperties, rendererIndex=1}
   {rendererClass=&NiagaraLightRendererProperties, rendererIndex=2}
  ]
 }
 Die {
  emitterSpawnScript [/SetVariables(Emitter.ThisEmittersArrangementsID:ENiagara_Papaya_FluorescentLightArrangements=::NewEnumerator6) {isSetParameters=true, module=SetVariables_#12}]
  emitterUpdateScript [
   &EmitterState()
   &self:GiveBirthToParticlesIfItsTheirTime() {module=GiveBirthToParticlesIfItsTheirTime}
  ]
  particleSpawnScript [
   &InitializeParticle(
    Module.Write Color=true
    Module.Write Lifetime=true
    Module.Write Mass=true
    Module.Write Material Random=true
    Module.Write Position=true
    Module.Write Scale=false
    Module.Write SpriteRotation=false
    Module.Write SpriteSize=false
    Position=<0 0 0>
    Sprite Size=<50 50>
   )
   &self:ReadSenderOtherParticleAttributes() {module=ReadSenderOtherParticleAttributes}
   /SetVariables(Particles.PercussionLookupIndex:NiagaraFloat=&UniformRangedFloat(
     Minimum=0
     Maximum=1
     Randomness Mode=::NewEnumerator1
     Random Seed=&Add_Integer(A=$Particles.GroupID, B=$Engine.System.TickCount)
     Fixed Random Seed=true
    )) {isSetParameters=true, module=SetVariables_#13}
   /SetVariables(Particles.RandomEnergyBurstMultiplier:NiagaraFloat=&UniformRangedFloat(Minimum=0, Maximum=1)) {isSetParameters=true, module=SetVariables_#14}
   /SetVariables(Particles.RandomValue:NiagaraFloat=&UniformRangedFloat(Minimum=0.5, Maximum=1.5)) {isSetParameters=true}
  ]
  particleUpdateScript [
   &self:MultiLerpPercussionSamples(PercussionLookupIndex:NiagaraFloat=$Particles.PercussionLookupIndex) {module=MultiLerpPercussionSamples}
   &self:PickRotationCenterVector(
    FrameCounter:NiagaraInt32=&UniformRangedInt(Minimum=0, Maximum=50000)
    GroupID:NiagaraInt32=&UniformRangedInt(Minimum=0, Maximum=50000)
    Rate Of Change:NiagaraFloat=&Multiply_Float(A=$Engine.DeltaTime, B=1)
   ) {module=PickRotationCenterVector001}
   &self:SelectCurrentExitTarget(
    "Actor Space Direction Vector  perp to Exit":Vector3f=<0 1 0>
    Target Impact Range=1600
    Top Pos:Vector3f=$Particles.Position_Top
    WS Exit Positions:NiagaraDataInterfaceArrayFloat3=$User.ExitLocations
   ) {module=SelectCurrentExitTarget}
   &self:TemporalFade(
    Current Value:NiagaraFloat=&Multiply_Float(A=0.5, B=&Add_Float(A=$Particles.PercussionValue, B=$NPC.Papaya_Jukebox_NPC.AverageAmplitude))
    Fade Rate=2
   ) {module=TemporalFade}
   &CurlNoiseForce(
    Noise Frequency=0.001
    Noise Strength=&Multiply_Float(A=0, B=&ClampFloat(Float=&Multiply_Float(A=$Particles.Age, B=1)))
    Pan Noise=true
    Pan Noise Field=<0.25 0.125 0.33>
    Sample Position:Vector3f=$Particles.Position_Top
   )
   &ParticleState(Kill Particles When Lifetime Has Elapsed=false)
   &self:SeeIfParticlesNeedToDie(Number Of Frames To Linger:NiagaraInt32=$System.Number Of Frames To Linger) {module=SeeIfParticlesNeedToDie}
   &SpringForce(
    " Particle Equilibrium Position":Vector3f=$Particles.Top_TargetLocation
    Desired Minimum Separation Distance=0
    Force Strength=0.5
    Particle Position:Vector3f=$Particles.Position_Top
    Particle Velocity:Vector3f=$Particles.Velocity_Top
    Spring Tightness:NiagaraFloat=&Multiply_Float(
     A=&Multiply_Float(A=$Particles.RandomValue, B=90)
     B=&ClampFloat(Float=&Multiply_Float(A=$Particles.Age, B=0.5))
    )
    Write to Intrinsic Parameters=false
   )
   &SpringForce(
    " Particle Equilibrium Position":Vector3f=$Particles.Position_Top
    Attachment Point Velocity:Vector3f=$Particles.Velocity_Top
    Dampening Coefficient=0.75
    Desired Minimum Separation Distance=0
    Particle Position:Vector3f=$Particles.Position_Bottom
    Particle Velocity:Vector3f=$Particles.Velocity_Bottom
    Spring Tightness:NiagaraFloat=&Multiply_Float(
     A=&Multiply_Float(A=$Particles.RandomValue, B=300)
     B=&ClampFloat(Float=&Multiply_Float(A=$Particles.Age, B=0.5))
    )
    Write to Intrinsic Parameters=false
   ) {module=SpringForce001}
   &self:Set_Top_Or_Bottom_Force(Bottom Force:Vector3f=$Output.SpringForce001.SpringForce, Top Force:Vector3f=$Output.SpringForce.SpringForce) {module=Set_Top_Or_Bottom_Force}
   &self:-DUPLICATE_BELOW_THIS_MODULE-__Names() {module=-DUPLICATE_BELOW_THIS_MODULE-__Names}
   &self:BlendInInheritedVelocityAtSpawn(Clamped Normalized Age Curve:NiagaraDataInterfaceCurve=$System.Transfer to New Velocity Curve) {module=BlendInInheritedVelocityAtSpawn}
   &SolveForcesAndVelocity(
    Acceleration Limit:NiagaraFloat=&Multiply_Float(A=$System.AccelerationLimit, B=2)
    Clamp Velocity=true
    Force:Vector3f=$Transient.Top_PhysicsForce
    Limit Acceleration=true
    Position:Vector3f=$Particles.Position_Top
    PreviousVelocity:Vector3f=$Particles.Previous.Velocity_Top
    Speed Limit:NiagaraFloat=&Multiply_Float(A=$System.MaxSpeed, B=2)
    Velocity:Vector3f=$Particles.Velocity_Top
    Write to Intrinsic Properties=false
   )
   &SolveForcesAndVelocity(
    Acceleration Limit:NiagaraFloat=&Multiply_Float(A=$System.AccelerationLimit, B=2)
    Clamp Velocity=true
    Force:Vector3f=$Transient.Bottom_PhysicsForce
    Limit Acceleration=true
    Position:Vector3f=$Particles.Position_Bottom
    PreviousVelocity:Vector3f=$Particles.Previous.Velocity_Bottom
    Speed Limit:NiagaraFloat=&Multiply_Float(A=$System.MaxSpeed, B=2)
    Velocity:Vector3f=$Particles.Velocity_Bottom
    Write to Intrinsic Properties=false
   ) {module=SolveForcesAndVelocity001}
   &self:StoreOffTopAndBottomVariables(
    Bottom Position:NiagaraPosition=$Output.SolveForcesAndVelocity001.Position
    Bottom Velocity:Vector3f=$Output.SolveForcesAndVelocity001.Velocity
    Top Position:NiagaraPosition=$Output.SolveForcesAndVelocity.Position
    Top Velocity:Vector3f=$Output.SolveForcesAndVelocity.Velocity
   ) {module=StoreOffTopAndBottomVariables}
   &self:InitialTransferAnimation_(
    Include Initialization Rotation Animation=false
    RandomAxisOfRotation:Vector3f=$Particles.PickRotationCenterVector001.RotationAxis
    Rotation Modulation:NiagaraFloat=$Particles.RandomEnergyBurstMultiplier
   ) {module=InitialTransferAnimation_}
   &self:FeedPositionsIntoShader(
    Bottom Position:Vector3f=$Particles.Position_Bottom
    Initial PercussionValue:NiagaraFloat=$Particles.Initial.PercussionValue
    PercussionValue:NiagaraFloat=$Particles.TemporalFade.CurrentValue
    Top Position:Vector3f=$Particles.Position_Top
   ) {module=FeedPositionsIntoShader}
   &SpriteBasedLine(Point A:Vector3f=$Particles.Position_Top, Point B:Vector3f=$Particles.Position_Bottom, SpriteWidth=0.01)
   &CalculateAccurateVelocity(
    Current Position:Vector3f=$Particles.Position_Top
    Previous Position:Vector3f=$Particles.Previous.Position_Top
    Set Intrinsic Variables=false
   ) {module=CalculateAccurateVelocity002}
   &CalculateAccurateVelocity(
    Current Position:Vector3f=$Particles.Position_Bottom
    Previous Position:Vector3f=$Particles.Previous.Position_Bottom
    Set Intrinsic Variables=false
   ) {module=CalculateAccurateVelocity003}
   /SetVariables(
    Particles.Position=$Particles.SpriteBasedLine.Position
    Particles.Velocity_Bottom:Vector3f=$Output.CalculateAccurateVelocity003.Velocity
    Particles.Velocity_Top:Vector3f=$Output.CalculateAccurateVelocity002.Velocity
   ) {isSetParameters=true, module=SetVariables_#15}
   &self:CalculateLightSize(
    Color:LinearColor=&MakeLinearColorFromVectorAndFloat(
     "Vector (RGB)"=&MakeVectorFromLinearColorRGB(Color=&MultiplyLinearColor(
       Linear Color=&ColorFromCurve(ColorCurve=$System.BaseColorCurve, CurveIndex=&Frac_Float(Float=$NPC.Papaya_Jukebox_NPC.Base Color Parameter))
       Scale Factor=&MakeVector4_FromFloat(Float=&Lerp_Float(A=0.5, B=2, Alpha=$Particles.TemporalFade.CurrentValue))
      ))
     "Float (Alpha)"=1
    )
    Sprite Based Line Size:Vector2f=$Particles.SpriteBasedLine.SpriteSize
   ) {module=CalculateLightSize}
  ]
  rendererClasses [
   &NiagaraMeshRendererProperties
   &NiagaraSpriteRendererProperties
   &NiagaraLightRendererProperties
  ]
  renderers [
   {rendererClass=&NiagaraMeshRendererProperties, rendererIndex=0}
   {rendererClass=&NiagaraSpriteRendererProperties, rendererIndex=1}
   {rendererClass=&NiagaraLightRendererProperties, rendererIndex=2}
  ]
 }
 Polyhedra {
  bEnabled=false
  emitterSpawnScript [
   /SetVariables(Emitter.ThisEmittersArrangementsID:ENiagara_Papaya_FluorescentLightArrangements=::NewEnumerator4) {isSetParameters=true, module=SetVariables_#12}
   /SetVariables(Emitter.ShapePointArray=DataInterface() {propertyValues {GpuSyncMode=SyncCpuToGpu, MaxElements=0, FloatData [<0 0 1>, <-1 0 0>, <0 1 0>, <1 0 0>, <0 -1 0>, <0 0 -1>]}}) {isSetParameters=true, module=SetVariables_#16}
   /SetVariables(Emitter.Edges=DataInterface() {propertyValues {GpuSyncMode=SyncCpuToGpu, MaxElements=0, FloatData [<0 1>, <0 2>, <0 3>, <0 4>, <1 2>, <2 3>, <3 4>, <4 1>, <1 5>, <2 5>, <3 5>, <4 5>]}}) {isSetParameters=true, module=SetVariables_#17}
  ]
  emitterUpdateScript [
   &EmitterState()
   &self:GiveBirthToParticlesIfItsTheirTime() {module=GiveBirthToParticlesIfItsTheirTime}
   &SpawnBurst_Instantaneous(Spawn Count=55, Use Spawn Probability=false)
   /Niagara/Modules/Update/Utility/EmitterFrameCounter()
  ]
  particleSpawnScript [
   &InitializeParticle(
    Module.Write Color=true
    Module.Write Lifetime=true
    Module.Write Mass=true
    Module.Write Material Random=true
    Module.Write Position=false
    Module.Write Scale=false
    Module.Write SpriteRotation=false
    Module.Write SpriteSize=false
    Sprite Size=<50 50>
   )
   &self:Form_Spike_Groups(Number Spheres:NiagaraInt32=$User.Number Of Spheres) {module=Form_Spike_Groups}
   &self:CalculateLocalTetrahedronPositions(
    New Vector 2D Array:NiagaraDataInterfaceArrayFloat2=$Emitter.Edges
    New Vector Array:NiagaraDataInterfaceArrayFloat3=$Emitter.ShapePointArray
   ) {module=CalculateLocalTetrahedronPositions}
   &self:Initial_Tetrahedron_Macro_Transform(
    Position Center:Vector3f=$User.PolyhedraCenter
    X Vector:Vector3f=$User.PolyhedraXAxis
    Y Vector:Vector3f=$User.PolyhedraYAxis
    groupID:NiagaraInt32=$Particles.GroupID
   ) {module=Initial_Tetrahedron_Macro_Transform}
   /NiagaraStaging/Unused/Modules/FibonacciSphere(Normalized Phase:NiagaraFloat=$Particles.Normalized Phase) {enabled=false}
   /SetVariables(Particles.Bottom_TargetLocation:Vector3f=&Add_Vector(A=$Particles.GroupCenter, B=$Particles.Bottom_TargetLocation)) {isSetParameters=true, module=SetVariables_#18}
   /SetVariables(Particles.Top_TargetLocation:Vector3f=&Add_Vector(A=$Particles.GroupCenter, B=$Particles.Top_TargetLocation)) {isSetParameters=true, module=SetVariables_#19}
   /SetVariables(Particles.LocalUnitSphere:Vector3f=$Output.FibonacciSphere.LocalVectorFromCenter) {enabled=false, isSetParameters=true, module=SetVariables_#20}
   &self:Find_FormationGroup_Centers(
    Cosine Offset Multiplier:Vector3f=$User.MultiSphereRightVector
    Position Offset:Vector3f=$User.MultiSphereCenter
    Sine Offset Multiplier:Vector3f=$User.MultiSphereUpVector
    Total Number of Groups:NiagaraInt32=$User.Number Of Spheres
    Total number of points per group:NiagaraInt32=$Transient.Total NumberOf Particles Per Sphere
   ) {enabled=false, module=Find_FormationGroup_Centers}
   &self:FindSphereTargetPositions(BottomDistance=10, Group Center:Vector3f=$Particles.GroupCenter, SphereUnitVector:Vector3f=$Particles.LocalUnitSphere, TopDistance=300) {enabled=false, module=FindSphereTargetPositions}
   &self:ReadSenderOtherParticleAttributes() {enabled=false, module=ReadSenderOtherParticleAttributes}
   /SetVariables() {enabled=false, isSetParameters=true}
  ]
  particleUpdateScript [
   /SetVariables(Particles.PercussionValue:NiagaraFloat=/Niagara/DynamicInputs/Math/LerpMultipleFloats(
     Float Count:ENiagara_MultipleLerpCount=::NewEnumerator1
     1:NiagaraFloat=&MakeFloatFromVector4(Vector 4=$NPC.Papaya_Jukebox_NPC.Percussion, Channel=::NewEnumerator0)
     2:NiagaraFloat=&MakeFloatFromVector4(Vector 4=$NPC.Papaya_Jukebox_NPC.Percussion, Channel=::NewEnumerator1)
     3:NiagaraFloat=&MakeFloatFromVector4(Vector 4=$NPC.Papaya_Jukebox_NPC.Percussion, Channel=::NewEnumerator2)
     4:NiagaraFloat=&MakeFloatFromVector4(Vector 4=$NPC.Papaya_Jukebox_NPC.Percussion, Channel=::NewEnumerator3)
     5:NiagaraFloat=?
     Alpha:NiagaraFloat=$Particles.PercussionLookupIndex
    )) {enabled=false, isSetParameters=true}
   &self:MultiLerpPercussionSamples(PercussionLookupIndex:NiagaraFloat=$Particles.PercussionLookupIndex) {enabled=false, module=MultiLerpPercussionSamples}
   &ParticleState(Kill Particles When Lifetime Has Elapsed=false)
   &self:SeeIfParticlesNeedToDie(Number Of Frames To Linger:NiagaraInt32=$System.Number Of Frames To Linger) {enabled=false, module=SeeIfParticlesNeedToDie}
   /SetVariables(Particles.EmitterFrameCount:NiagaraInt32=$Emitter.EmitterFrameCounter.Counter) {enabled=false, isSetParameters=true}
   &self:PickRotationCenterVector(
    FrameCounter:NiagaraInt32=$Emitter.EmitterFrameCounter.Counter
    GroupID:NiagaraInt32=$Particles.Sphere Group
    Rate Of Change:NiagaraFloat=&Multiply_Float(A=$Engine.DeltaTime, B=1)
   ) {enabled=false, module=PickRotationCenterVector}
   &self:RotateLocalUnitSphere(RotationAxis:Vector3f=$Particles.PickRotationCenterVector.RotationAxis) {enabled=false, module=RotateLocalUnitSphere002}
   &self:FindSphereTargetPositions(BottomDistance=100, Group Center:Vector3f=$Particles.GroupCenter, SphereUnitVector:Vector3f=$Particles.LocalUnitSphere, TopDistance=200) {enabled=false, module=FindSphereTargetPositions001}
   &CalculateAccurateVelocity(
    Current Position:Vector3f=$Particles.Target_Position_Bottom
    Previous Position:Vector3f=$Particles.Previous.Target_Position_Bottom
    Set Intrinsic Variables=false
   ) {enabled=false}
   &CalculateAccurateVelocity(
    Current Position:Vector3f=$Particles.Target_Position_Top
    Previous Position:Vector3f=$Particles.Previous.Target_Position_Top
    Set Intrinsic Variables=false
   ) {enabled=false, module=CalculateAccurateVelocity001}
   &SpringForce(
    " Particle Equilibrium Position":Vector3f=$Particles.Target_Position_Top
    Attachment Point Velocity:Vector3f=$Output.CalculateAccurateVelocity001.Velocity
    Dampening Coefficient=10.9034
    Desired Minimum Separation Distance=0
    Particle Position:Vector3f=$Particles.Position_Top
    Particle Velocity:Vector3f=$Particles.Velocity_Top
    Spring Tightness=32.2028
    Write to Intrinsic Parameters=false
   ) {enabled=false}
   &SpringForce(
    " Particle Equilibrium Position":Vector3f=$Particles.Target_Position_Bottom
    Attachment Point Velocity:Vector3f=$Output.CalculateAccurateVelocity.Velocity
    Dampening Coefficient=7.2428
    Desired Minimum Separation Distance=0
    Particle Position:Vector3f=$Particles.Position_Bottom
    Particle Velocity:Vector3f=$Particles.Velocity_Bottom
    Spring Tightness=44.0396
    Write to Intrinsic Parameters=false
   ) {enabled=false, module=SpringForce001}
   &self:Set_Top_Or_Bottom_Force(Bottom Force:Vector3f=$Output.SpringForce001.SpringForce, Top Force:Vector3f=$Output.SpringForce.SpringForce) {enabled=false, module=Set_Top_Or_Bottom_Force}
   &self:-DUPLICATE_BELOW_THIS_MODULE-__Names() {enabled=false, module=-DUPLICATE_BELOW_THIS_MODULE-__Names}
   &self:BlendInInheritedVelocityAtSpawn(Clamped Normalized Age Curve:NiagaraDataInterfaceCurve=$System.Transfer to New Velocity Curve) {enabled=false, module=BlendInInheritedVelocityAtSpawn}
   &SolveForcesAndVelocity(
    Acceleration Limit:NiagaraFloat=$System.AccelerationLimit
    Clamp Velocity=true
    Force:Vector3f=$Transient.Top_PhysicsForce
    Limit Acceleration=true
    Position:Vector3f=$Particles.Position_Top
    PreviousVelocity:Vector3f=$Particles.Previous.Velocity_Top
    Speed Limit:NiagaraFloat=$System.MaxSpeed
    Velocity:Vector3f=$Particles.Velocity_Top
    Write to Intrinsic Properties=false
   ) {enabled=false}
   &SolveForcesAndVelocity(
    Acceleration Limit:NiagaraFloat=$System.AccelerationLimit
    Clamp Velocity=true
    Force:Vector3f=$Transient.Bottom_PhysicsForce
    Limit Acceleration=true
    Position:Vector3f=$Particles.Position_Bottom
    PreviousVelocity:Vector3f=$Particles.Previous.Velocity_Bottom
    Speed Limit:NiagaraFloat=$System.MaxSpeed
    Velocity:Vector3f=$Particles.Velocity_Bottom
    Write to Intrinsic Properties=false
   ) {enabled=false, module=SolveForcesAndVelocity001}
   &self:StoreOffTopAndBottomVariables(
    Bottom Position:NiagaraPosition=$Output.SolveForcesAndVelocity001.Position
    Bottom Velocity:Vector3f=$Output.SolveForcesAndVelocity001.Velocity
    Top Position:NiagaraPosition=$Output.SolveForcesAndVelocity.Position
    Top Velocity:Vector3f=$Output.SolveForcesAndVelocity.Velocity
   ) {enabled=false, module=StoreOffTopAndBottomVariables}
   &self:InitialTransferAnimation_() {enabled=false, module=InitialTransferAnimation_}
   &self:FeedPositionsIntoShader(
    Bottom Position:Vector3f=$Particles.Position_Bottom
    Initial PercussionValue:NiagaraFloat=$Particles.Initial.PercussionValue
    PercussionValue:NiagaraFloat=$Particles.PercussionValue
    Top Position:Vector3f=$Particles.Position_Top
   ) {enabled=false, module=FeedPositionsIntoShader}
   &self:FeedPositionsIntoShader(
    Bottom Position:Vector3f=$Particles.Bottom_TargetLocation
    Initial PercussionValue:NiagaraFloat=$Particles.Initial.PercussionValue
    PercussionValue:NiagaraFloat=$Particles.PercussionValue
    Top Position:Vector3f=$Particles.Top_TargetLocation
   ) {module=FeedPositionsIntoShader001}
   &SpriteBasedLine(Point A:Vector3f=$Particles.Position_Top, Point B:Vector3f=$Particles.Position_Bottom, SpriteWidth=0.01) {enabled=false}
   &CalculateAccurateVelocity(
    Current Position:Vector3f=$Particles.Position_Top
    Previous Position:Vector3f=$Particles.Previous.Position_Top
    Set Intrinsic Variables=false
   ) {enabled=false, module=CalculateAccurateVelocity002}
   &CalculateAccurateVelocity(
    Current Position:Vector3f=$Particles.Position_Bottom
    Previous Position:Vector3f=$Particles.Previous.Position_Bottom
    Set Intrinsic Variables=false
   ) {enabled=false, module=CalculateAccurateVelocity003}
   /SetVariables(
    Particles.Position=$Particles.SpriteBasedLine.Position
    Particles.Velocity_Bottom:Vector3f=$Output.CalculateAccurateVelocity003.Velocity
    Particles.Velocity_Top:Vector3f=$Output.CalculateAccurateVelocity002.Velocity
   ) {enabled=false, isSetParameters=true, module=SetVariables_#15}
   &self:CalculateLightSize(Sprite Based Line Size:Vector2f=$Particles.SpriteBasedLine.SpriteSize) {module=CalculateLightSize}
  ]
  rendererClasses [
   &NiagaraMeshRendererProperties
   &NiagaraSpriteRendererProperties
   &NiagaraLightRendererProperties
  ]
  renderers [
   {rendererClass=&NiagaraMeshRendererProperties, rendererIndex=0}
   {rendererClass=&NiagaraSpriteRendererProperties, rendererIndex=1}
   {rendererClass=&NiagaraLightRendererProperties, rendererIndex=2}
  ]
 }
 RecursiveLightEmitter {
  bEnabled=false
  emitterUpdateScript [
   &EmitterState()
   &SpawnBurst_Instantaneous(Spawn Count=1, Use Spawn Probability=false)
  ]
  particleSpawnScript [&InitializeParticle(
    Lifetime=$Emitter.CurrentLoopDuration
    Mesh Scale=<30 300 5>
    Module.Write Color=true
    Module.Write Lifetime=true
    Module.Write Mass=true
    Module.Write Material Random=true
    Module.Write Position=false
    Module.Write Scale=true
    Module.Write SpriteRotation=false
    Module.Write SpriteSize=true
    Sprite Size=<50 50>
   )]
  particleUpdateScript [
   &ParticleState()
   /Niagara/Modules/Update/Utility/IncrementOverTime()
   &self:Send_Recursive_Event(Incrementing value over time:NiagaraFloat=$Particles.IncrementOverTime.Counter, Max EmissionHierarchical Depth=3, Max Number of spawns=1) {module=Send_Recursive_Event}
   &SolveForcesAndVelocity(Clamp Velocity=false, Limit Acceleration=false)
  ]
  rendererClasses [&NiagaraMeshRendererProperties]
  renderers [{rendererClass=&NiagaraMeshRendererProperties, rendererIndex=0}]
 }
 SimpleSpriteBurst {
  bEnabled=false
  emitterUpdateScript [
   &EmitterState(Life Cycle Mode=::NewEnumerator1, Loop Duration=1.25, UseLoopDelay=false)
   &SpawnBurst_Instantaneous(Spawn Count=1, Use Spawn Probability=false)
  ]
  particleSpawnScript [&InitializeParticle(
    Module.Write Color=true
    Module.Write Lifetime=true
    Module.Write Mass=true
    Module.Write Material Random=true
    Module.Write Position=false
    Module.Write Scale=false
    Module.Write SpriteRotation=false
    Module.Write SpriteSize=true
    Sprite Size=<500 500>
   )]
  particleUpdateScript [
   &ParticleState()
   &DynamicMaterialParameters(
    Index 0 Param 1=&MakeFloatFromInt(Int=$System.NumberOfLiveParticlesInSystem)
    Module.Param0WriteEnabled=true
    Module.Param1WriteEnabled=true
    Module.Param2WriteEnabled=true
    Module.Param3WriteEnabled=true
   )
   &SolveForcesAndVelocity(Clamp Velocity=false, Limit Acceleration=false)
  ]
  rendererClasses [&NiagaraSpriteRendererProperties]
  renderers [{rendererClass=&NiagaraSpriteRendererProperties}]
 }
 SpawnLine {
  emitterSpawnScript [/SetVariables() {isSetParameters=true}]
  emitterUpdateScript [
   &EmitterState()
   /SetVariables(Emitter.InitializeBrandNewParticles:NiagaraBool=&Boolean_AndOperation(Bool1=$System.BPCallsForActivationCall, Bool2=&Boolean_NotOperation(Bool=$System.SystemContainsParticles))) {isSetParameters=true}
   &self:GiveBirthToParticlesIfItsTheirTime(ForceSpawnBurst:NiagaraBool=$Emitter.InitializeBrandNewParticles) {module=GiveBirthToParticlesIfItsTheirTime}
  ]
  particleSpawnScript [
   &InitializeParticle(
    Module.Write Color=true
    Module.Write Lifetime=true
    Module.Write Mass=true
    Module.Write Material Random=true
    Module.Write Position=false
    Module.Write Scale=false
    Module.Write SpriteRotation=false
    Module.Write SpriteSize=true
    Sprite Size=<50 50>
   )
   /SetVariables(Particles.LightRadius=1200) {isSetParameters=true}
   /SetVariables() {isSetParameters=true}
   &self:Arrange_Particles_in_A_Curly_Line(
    Cosine Influence:Vector3f=&MakeVector(
     X=&FloatFromCurve(
      FloatCurve=DataInterface() {propertyValues {Curve {keys [<at 0.1623 1>]}}}
      CurveIndex=&ReturnNormalizedExecIndex()
      Scale Curve=10
     )
     Y=0
     Z=0
    )
    Linear Influnece:Vector3f=&Multiply_VectorByFloat(
     Vector=&VectorFromCurve(
      VectorCurve=DataInterface() {propertyValues {
        XCurve {keys [
          {arriveTangent=2.0906, interpMode=RCIM_Cubic, leaveTangent=2.0906, tangentMode=RCTM_User, time=0, value=0}
          {arriveTangent=1.802, interpMode=RCIM_Cubic, leaveTangent=1.802, time=0.1956, value=0.9228}
          {arriveTangent=-0.0231, interpMode=RCIM_Cubic, leaveTangent=-0.0231, time=0.5, value=0.901}
          {arriveTangent=-1.802, interpMode=RCIM_Cubic, leaveTangent=-1.802, time=0.8226, value=0.9083}
          {arriveTangent=-2.5014, interpMode=RCIM_Cubic, leaveTangent=-2.5014, tangentMode=RCTM_User, time=1, value=0}
         ]}
        YCurve {keys [<at 0 0>,<at 1 1>]}
        ZCurve {keys [<at 0 0>,<at 1 0>]}
       }}
      CurveIndex=&ReturnNormalizedExecIndex()
     )
     Float=1
    )
    Local Space Object Center:Vector3f=$User.LocalSpaceSpawnLocation
    Max Linear Influence Offset:Vector3f=<533 5077.4268 0>
    Rotation Phase Curve=DataInterface() {propertyValues {Curve {keys [
        {arriveTangent=45.5382, time=0, value=0}
        {arriveTangent=-248670.7188, interpMode=RCIM_Cubic, leaveTangent=-248670.8125, tangentMode=RCTM_User, time=1, value=3.2916}
       ]}}}
    Sine Influence:Vector3f=&MakeVector(
     X=&FloatFromCurve(
      FloatCurve=DataInterface() {propertyValues {Curve {keys [<at 0.0173 1>]}}}
      CurveIndex=&ReturnNormalizedExecIndex()
      Scale Curve=200
     )
     Y=0
     Z=0
    )
   ) {module=Arrange_Particles_in_A_Curly_Line}
   &self:ReadSenderOtherParticleAttributes(
    Position:Vector3f=$Output.Line Position
    Position_Bottom:Vector3f=$Output.Line Position
    Position_Top:Vector3f=$Output.Line Position
    ReceiveTransferredParticleAttributes:NiagaraBool=&Boolean_NotOperation(Bool=$Emitter.InitializeBrandNewParticles)
   ) {module=ReadSenderOtherParticleAttributes}
   /SetVariables(
    Particles.Lag_Second_Point_Behind.TemporalLerp_Vector.Moving Average:Vector3f=$Particles.Position_Bottom
    Particles.Position_Bottom_MovingAverage:Vector3f=$Particles.Position_Bottom
   ) {isSetParameters=true}
   /SetVariables(Particles.Top_TargetLocation:Vector3f=&Add_Vector(
     A=$Output.Line Position
     B=&MakeVector(
      X=0
      Y=0
      Z=&Lerp_Float(
       A=1000
       B=1886.7584
       Alpha=&FloatFromCurve(
        FloatCurve=DataInterface() {propertyValues {Curve {keys [<at 0 0 RCIM_Cubic>,<at 0.4924 0.8673 RCIM_Cubic>,<at 1 0 RCIM_Cubic>]}}}
        CurveIndex=&ReturnNormalizedExecIndex()
       )
      )
     )
    )) {isSetParameters=true}
   /SetVariables(Particles.Bottom_TargetLocation:Vector3f=&Add_Vector(
     A=$Output.Line Position
     B=&MakeVector(
      X=0
      Y=0
      Z=&Lerp_Float(
       A=1000
       B=772.3597
       Alpha=&FloatFromCurve(
        FloatCurve=DataInterface() {propertyValues {Curve {keys [
            <at -0.0064 -0.0009 RCIM_Cubic>
            <at 0.2362 1.4093>
            <at 0.3725 1.9382>
            <at 0.4487 0.7941>
            {arriveTangent=-0.0007, interpMode=RCIM_Cubic, leaveTangent=-0.0007, time=0.4984, value=1.9757}
            <at 0.5731 0.794>
            <at 0.6259 1.9382>
            <at 0.745 1.3501>
            <at 1.0046 0.0319 RCIM_Cubic>
           ]}}}
        CurveIndex=&ReturnNormalizedExecIndex()
       )
      )
     )
    )) {isSetParameters=true}
   /SetVariables(Particles.PercussionLookupIndex:NiagaraFloat=&FloatFromCurve(
     FloatCurve=DataInterface() {propertyValues {Curve {keys [<at 0 1 RCIM_Cubic>,<at 0.5006 0.0048>,<at 1 1 RCIM_Cubic>]}}}
     CurveIndex=&ReturnNormalizedExecIndex()
    )) {isSetParameters=true}
   /SetVariables(Particles.RandomEnergyBurstMultiplier:NiagaraFloat=&UniformRangedFloat(Minimum=0, Maximum=2)) {isSetParameters=true}
  ]
  particleUpdateScript [
   &self:MultiLerpPercussionSamples(PercussionLookupIndex:NiagaraFloat=$Particles.PercussionLookupIndex) {module=MultiLerpPercussionSamples}
   &self:PickRotationCenterVector(
    FrameCounter:NiagaraInt32=&UniformRangedInt(Minimum=0, Maximum=50000)
    GroupID:NiagaraInt32=&UniformRangedInt(Minimum=0, Maximum=50000)
    Rate Of Change:NiagaraFloat=&Multiply_Float(A=$Engine.DeltaTime, B=1)
   ) {module=PickRotationCenterVector001}
   /SetVariables(Particles.PercussionValue:NiagaraFloat=&LerpMultipleFloats(
     Float Count:ENiagara_MultipleLerpCount=::NewEnumerator1
     1:NiagaraFloat=&MakeFloatFromVector4(Vector 4=$NPC.Papaya_Jukebox_NPC.Percussion, Channel=::NewEnumerator0)
     2:NiagaraFloat=&MakeFloatFromVector4(Vector 4=$NPC.Papaya_Jukebox_NPC.Percussion, Channel=::NewEnumerator1)
     3:NiagaraFloat=&MakeFloatFromVector4(Vector 4=$NPC.Papaya_Jukebox_NPC.Percussion, Channel=::NewEnumerator2)
     4:NiagaraFloat=&MakeFloatFromVector4(Vector 4=$NPC.Papaya_Jukebox_NPC.Percussion, Channel=::NewEnumerator3)
     5:NiagaraFloat=?
     Alpha:NiagaraFloat=$Particles.PercussionLookupIndex
    )) {enabled=false, isSetParameters=true}
   &ParticleState(Kill Particles When Lifetime Has Elapsed=false)
   &self:SeeIfParticlesNeedToDie(Number Of Frames To Linger:NiagaraInt32=$System.Number Of Frames To Linger) {module=SeeIfParticlesNeedToDie}
   /Niagara/Modules/Update/Utility/FrameCounter() {enabled=false}
   &CurlNoiseForce(
    Noise Frequency=0.0017
    Noise Strength=500
    Pan Noise=true
    Pan Noise Field=<0.25 0.25 0.125>
    Sample Position:Vector3f=$Particles.Position_Top
    Write to Intrinsic Parameters=false
   )
   &CurlNoiseForce(
    Noise Frequency=0.0017
    Noise Strength=500
    Pan Noise=true
    Pan Noise Field=<1 1.1046 2.261>
    Sample Position:Vector3f=$Particles.Position_Bottom
    Write to Intrinsic Parameters=false
   ) {module=CurlNoiseForce001}
   &Drag(Use Linear Drag=true, Use Rotational Drag=false)
   &SpringForce(
    " Particle Equilibrium Position":NiagaraPosition=&Add_Vector(
     A=$Particles.Top_TargetLocation
     B=&MakeVector(X=0, Y=0, Z=&Multiply_Float(A=$Particles.PercussionValue, B=600))
    )
    Attachment Point Velocity:Vector3f=$Output.CalculateAccurateVelocity.Velocity
    Dampening Coefficient=2
    Desired Minimum Separation Distance=0
    Particle Position:Vector3f=$Particles.Position_Top
    Particle Velocity:Vector3f=$Particles.Velocity_Top
    Spring Tightness:NiagaraFloat=&Lerp_Float(A=50, B=600, Alpha=&ClampFloat(Float=&Divide_Float(A=$Particles.Age, B=5)))
    Write to Intrinsic Parameters=false
   )
   &SpringForce(
    " Particle Equilibrium Position":NiagaraPosition=&Add_Vector(
     A=$Particles.Bottom_TargetLocation
     B=&MakeVector(X=0, Y=0, Z=&Multiply_Float(A=$Particles.PercussionValue, B=200))
    )
    Attachment Point Velocity:Vector3f=$Output.CalculateAccurateVelocity001.Velocity
    Dampening Coefficient=2
    Desired Minimum Separation Distance=0
    Particle Position:Vector3f=$Particles.Position_Bottom
    Particle Velocity:Vector3f=$Particles.Velocity_Bottom
    Write to Intrinsic Parameters=false
   ) {module=SpringForce001}
   &self:Set_Top_Or_Bottom_Force(
    Bottom Force:Vector3f=&Add_Vector(A=$Output.SpringForce001.SpringForce, B=$Output.CurlNoiseForce001.CurlNoiseForce)
    Top Force:Vector3f=&Add_Vector(A=$Output.SpringForce.SpringForce, B=$Output.CurlNoiseForce.CurlNoiseForce)
   ) {module=Set_Top_Or_Bottom_Force}
   &self:-DUPLICATE_BELOW_THIS_MODULE-__Names() {module=-DUPLICATE_BELOW_THIS_MODULE-__Names}
   &self:BlendInInheritedVelocityAtSpawn(Clamped Normalized Age Curve:NiagaraDataInterfaceCurve=$System.Transfer to New Velocity Curve) {module=BlendInInheritedVelocityAtSpawn}
   &SolveForcesAndVelocity(
    Acceleration Limit:NiagaraFloat=$System.AccelerationLimit
    Clamp Velocity=true
    Force:Vector3f=$Transient.Top_PhysicsForce
    Limit Acceleration=true
    Position:Vector3f=$Particles.Position_Top
    PreviousVelocity:Vector3f=$Particles.Previous.Velocity_Top
    Speed Limit:NiagaraFloat=$System.MaxSpeed
    Velocity:Vector3f=$Particles.Velocity_Top
    Write to Intrinsic Properties=false
   )
   &SolveForcesAndVelocity(
    Acceleration Limit:NiagaraFloat=$System.AccelerationLimit
    Clamp Velocity=true
    Force:Vector3f=$Transient.Bottom_PhysicsForce
    Limit Acceleration=true
    Position:Vector3f=$Particles.Position_Bottom
    PreviousVelocity:Vector3f=$Particles.Previous.Velocity_Bottom
    Speed Limit:NiagaraFloat=$System.MaxSpeed
    Velocity:Vector3f=$Particles.Velocity_Bottom
    Write to Intrinsic Properties=false
   ) {module=SolveForcesAndVelocity001}
   &self:StoreOffTopAndBottomVariables(
    Bottom Position:NiagaraPosition=$Output.SolveForcesAndVelocity001.Position
    Bottom Velocity:Vector3f=$Output.SolveForcesAndVelocity001.Velocity
    Top Position:NiagaraPosition=$Output.SolveForcesAndVelocity.Position
    Top Velocity:Vector3f=$Output.SolveForcesAndVelocity.Velocity
   ) {module=StoreOffTopAndBottomVariables}
   &self:InitialTransferAnimation_(
    Include Initialization Rotation Animation=false
    RandomAxisOfRotation:Vector3f=$Particles.PickRotationCenterVector001.RotationAxis
    Rotation Modulation:NiagaraFloat=$Particles.RandomEnergyBurstMultiplier
   ) {module=InitialTransferAnimation_}
   &self:FeedPositionsIntoShader(
    Bottom Position:Vector3f=$Particles.Position_Bottom
    Initial PercussionValue:NiagaraFloat=$Particles.Initial.PercussionValue
    PercussionValue:NiagaraFloat=$Particles.PercussionValue
    Top Position:Vector3f=$Particles.Position_Top
   ) {module=FeedPositionsIntoShader}
   &SpriteBasedLine(Point A:Vector3f=$Particles.Position_Top, Point B:Vector3f=$Particles.Position_Bottom, SpriteWidth=0.01)
   &CalculateAccurateVelocity(
    Current Position:Vector3f=$Particles.Position_Top
    Previous Position:Vector3f=$Particles.Previous.Position_Top
    Set Intrinsic Variables=false
   ) {module=CalculateAccurateVelocity002}
   &CalculateAccurateVelocity(
    Current Position:Vector3f=$Particles.Position_Bottom
    Previous Position:Vector3f=$Particles.Previous.Position_Bottom
    Set Intrinsic Variables=false
   ) {module=CalculateAccurateVelocity003}
   /SetVariables(
    Particles.Position=$Particles.SpriteBasedLine.Position
    Particles.Velocity_Bottom:Vector3f=$Output.CalculateAccurateVelocity003.Velocity
    Particles.Velocity_Top:Vector3f=$Output.CalculateAccurateVelocity002.Velocity
   ) {isSetParameters=true}
   &self:CalculateLightSize(
    Color:LinearColor=&MakeLinearColorFromVectorAndFloat(
     "Vector (RGB)"=&MakeVectorFromLinearColorRGB(Color=&ColorFromCurve(ColorCurve=$System.BaseColorCurve, CurveIndex=&Frac_Float(Float=$NPC.Papaya_Jukebox_NPC.Base Color Parameter)))
     "Float (Alpha)"=1
    )
    Sprite Based Line Size:Vector2f=$Particles.SpriteBasedLine.SpriteSize
   ) {module=CalculateLightSize}
  ]
  rendererClasses [
   &NiagaraSpriteRendererProperties
   &NiagaraMeshRendererProperties
   &NiagaraLightRendererProperties
  ]
  renderers [
   {rendererClass=&NiagaraSpriteRendererProperties}
   {rendererClass=&NiagaraMeshRendererProperties, rendererIndex=1}
   {rendererClass=&NiagaraLightRendererProperties, rendererIndex=2}
  ]
 }
 Spheres {
  emitterSpawnScript [/SetVariables(Emitter.ThisEmittersArrangementsID:ENiagara_Papaya_FluorescentLightArrangements=::NewEnumerator2) {isSetParameters=true, module=SetVariables_#12}]
  emitterUpdateScript [
   &EmitterState()
   &self:GiveBirthToParticlesIfItsTheirTime() {module=GiveBirthToParticlesIfItsTheirTime}
  ]
  particleSpawnScript [
   &InitializeParticle(
    Module.Write Color=true
    Module.Write Lifetime=true
    Module.Write Mass=true
    Module.Write Material Random=true
    Module.Write Position=false
    Module.Write Scale=false
    Module.Write SpriteRotation=false
    Module.Write SpriteSize=false
    Sprite Size=<50 50>
   )
   &self:Form_Spike_Groups(Number Spheres:NiagaraInt32=$User.Number Of Spheres) {module=Form_Spike_Groups}
   &FibonacciSphere(Normalized Phase:NiagaraFloat=$Particles.Normalized Phase)
   /SetVariables(Particles.LocalUnitSphere:Vector3f=$Output.FibonacciSphere.LocalVectorFromCenter) {isSetParameters=true, module=SetVariables_#20}
   &self:Find_FormationGroup_Centers(
    Cosine Offset Multiplier:Vector3f=$User.MultiSphereRightVector
    Position Offset:Vector3f=$User.MultiSphereCenter
    Sine Offset Multiplier:Vector3f=$User.MultiSphereUpVector
    Total Number of Groups:NiagaraInt32=$User.Number Of Spheres
    Total number of points per group:NiagaraInt32=$Transient.Total NumberOf Particles Per Sphere
   ) {module=Find_FormationGroup_Centers}
   &self:FindSphereTargetPositions(BottomDistance=10, Group Center:Vector3f=$Particles.GroupCenter, SphereUnitVector:Vector3f=$Particles.LocalUnitSphere, TopDistance=300) {module=FindSphereTargetPositions}
   &self:ReadSenderOtherParticleAttributes() {module=ReadSenderOtherParticleAttributes}
   /SetVariables(Particles.PercussionLookupIndex:NiagaraFloat=$Output.NormalizedGroupID) {isSetParameters=true}
   /SetVariables(Particles.RandomEnergyBurstMultiplier:NiagaraFloat=&UniformRangedFloat(Minimum=0, Maximum=1)) {isSetParameters=true}
  ]
  particleUpdateScript [
   &self:MultiLerpPercussionSamples(PercussionLookupIndex:NiagaraFloat=$Particles.PercussionLookupIndex) {module=MultiLerpPercussionSamples}
   &self:PickRotationCenterVector(
    FrameCounter:NiagaraInt32=&UniformRangedInt(Minimum=0, Maximum=50000)
    GroupID:NiagaraInt32=&UniformRangedInt(Minimum=0, Maximum=50000)
    Rate Of Change:NiagaraFloat=&Multiply_Float(A=$Engine.DeltaTime, B=1)
   ) {module=PickRotationCenterVector001}
   &ParticleState(Kill Particles When Lifetime Has Elapsed=false)
   &self:SeeIfParticlesNeedToDie(Number Of Frames To Linger:NiagaraInt32=$System.Number Of Frames To Linger) {module=SeeIfParticlesNeedToDie}
   &self:PickRotationCenterVector(
    FrameCounter:NiagaraInt32=$Emitter.EmitterFrameCounter.Counter
    GroupID:NiagaraInt32=$Particles.Sphere Group
    Rate Of Change:NiagaraFloat=&Multiply_Float(A=$Engine.DeltaTime, B=1)
   ) {module=PickRotationCenterVector}
   &self:RotateLocalUnitSphere(RotationAxis:Vector3f=$Particles.PickRotationCenterVector.RotationAxis) {module=RotateLocalUnitSphere002}
   &self:FindSphereTargetPositions(BottomDistance=100, Group Center:Vector3f=$Particles.GroupCenter, SphereUnitVector:Vector3f=$Particles.LocalUnitSphere, TopDistance=200) {module=FindSphereTargetPositions001}
   &CalculateAccurateVelocity(
    Current Position:Vector3f=$Particles.Target_Position_Bottom
    Previous Position:Vector3f=$Particles.Previous.Target_Position_Bottom
    Set Intrinsic Variables=false
   )
   &CalculateAccurateVelocity(
    Current Position:Vector3f=$Particles.Target_Position_Top
    Previous Position:Vector3f=$Particles.Previous.Target_Position_Top
    Set Intrinsic Variables=false
   ) {module=CalculateAccurateVelocity001}
   &SpringForce(
    " Particle Equilibrium Position":Vector3f=$Particles.Target_Position_Top
    Attachment Point Velocity:Vector3f=$Output.CalculateAccurateVelocity001.Velocity
    Dampening Coefficient=10.9034
    Desired Minimum Separation Distance=0
    Particle Position:Vector3f=$Particles.Position_Top
    Particle Velocity:Vector3f=$Particles.Velocity_Top
    Spring Tightness=180
    Write to Intrinsic Parameters=false
   )
   &SpringForce(
    " Particle Equilibrium Position":Vector3f=$Particles.Target_Position_Bottom
    Attachment Point Velocity:Vector3f=$Output.CalculateAccurateVelocity.Velocity
    Dampening Coefficient=7.2428
    Desired Minimum Separation Distance=0
    Particle Position:Vector3f=$Particles.Position_Bottom
    Particle Velocity:Vector3f=$Particles.Velocity_Bottom
    Spring Tightness=200
    Write to Intrinsic Parameters=false
   ) {module=SpringForce001}
   &self:Set_Top_Or_Bottom_Force(Bottom Force:Vector3f=$Output.SpringForce001.SpringForce, Top Force:Vector3f=$Output.SpringForce.SpringForce) {module=Set_Top_Or_Bottom_Force}
   &self:-DUPLICATE_BELOW_THIS_MODULE-__Names() {module=-DUPLICATE_BELOW_THIS_MODULE-__Names}
   &self:BlendInInheritedVelocityAtSpawn(Clamped Normalized Age Curve:NiagaraDataInterfaceCurve=$System.Transfer to New Velocity Curve) {module=BlendInInheritedVelocityAtSpawn}
   &SolveForcesAndVelocity(
    Acceleration Limit:NiagaraFloat=$System.AccelerationLimit
    Clamp Velocity=true
    Force:Vector3f=$Transient.Top_PhysicsForce
    Limit Acceleration=true
    Position:Vector3f=$Particles.Position_Top
    PreviousVelocity:Vector3f=$Particles.Previous.Velocity_Top
    Speed Limit:NiagaraFloat=$System.MaxSpeed
    Velocity:Vector3f=$Particles.Velocity_Top
    Write to Intrinsic Properties=false
   )
   &SolveForcesAndVelocity(
    Acceleration Limit:NiagaraFloat=$System.AccelerationLimit
    Clamp Velocity=true
    Force:Vector3f=$Transient.Bottom_PhysicsForce
    Limit Acceleration=true
    Position:Vector3f=$Particles.Position_Bottom
    PreviousVelocity:Vector3f=$Particles.Previous.Velocity_Bottom
    Speed Limit:NiagaraFloat=$System.MaxSpeed
    Velocity:Vector3f=$Particles.Velocity_Bottom
    Write to Intrinsic Properties=false
   ) {module=SolveForcesAndVelocity001}
   &self:StoreOffTopAndBottomVariables(
    Bottom Position:NiagaraPosition=$Output.SolveForcesAndVelocity001.Position
    Bottom Velocity:Vector3f=$Output.SolveForcesAndVelocity001.Velocity
    Top Position:NiagaraPosition=$Output.SolveForcesAndVelocity.Position
    Top Velocity:Vector3f=$Output.SolveForcesAndVelocity.Velocity
   ) {module=StoreOffTopAndBottomVariables}
   &self:InitialTransferAnimation_(
    Include Initialization Rotation Animation=false
    RandomAxisOfRotation:Vector3f=$Particles.PickRotationCenterVector001.RotationAxis
    Rotation Modulation:NiagaraFloat=$Particles.RandomEnergyBurstMultiplier
   ) {module=InitialTransferAnimation_}
   &self:FeedPositionsIntoShader(
    Bottom Position:Vector3f=$Particles.Position_Bottom
    Initial PercussionValue:NiagaraFloat=$Particles.Initial.PercussionValue
    PercussionValue:NiagaraFloat=$Particles.PercussionValue
    Top Position:Vector3f=$Particles.Position_Top
   ) {module=FeedPositionsIntoShader}
   &SpriteBasedLine(Point A:Vector3f=$Particles.Position_Top, Point B:Vector3f=$Particles.Position_Bottom, SpriteWidth=0.01)
   &CalculateAccurateVelocity(
    Current Position:Vector3f=$Particles.Position_Top
    Previous Position:Vector3f=$Particles.Previous.Position_Top
    Set Intrinsic Variables=false
   ) {module=CalculateAccurateVelocity002}
   &CalculateAccurateVelocity(
    Current Position:Vector3f=$Particles.Position_Bottom
    Previous Position:Vector3f=$Particles.Previous.Position_Bottom
    Set Intrinsic Variables=false
   ) {module=CalculateAccurateVelocity003}
   /SetVariables(
    Particles.Position=$Particles.SpriteBasedLine.Position
    Particles.Velocity_Bottom:Vector3f=$Output.CalculateAccurateVelocity003.Velocity
    Particles.Velocity_Top:Vector3f=$Output.CalculateAccurateVelocity002.Velocity
   ) {isSetParameters=true, module=SetVariables_#15}
   &self:CalculateLightSize(
    Color:LinearColor=&MakeLinearColorFromVectorAndFloat(
     "Vector (RGB)"=&MakeVectorFromLinearColorRGB(Color=&MultiplyLinearColor(
       Linear Color=&ColorFromCurve(ColorCurve=$System.BaseColorCurve, CurveIndex=&Frac_Float(Float=$NPC.Papaya_Jukebox_NPC.Base Color Parameter))
       Scale Factor=&MakeVector4_FromFloat(Float=&Lerp_Float(A=0.5, B=2, Alpha=$Particles.PercussionValue))
      ))
     "Float (Alpha)"=1
    )
    Sprite Based Line Size:Vector2f=$Particles.SpriteBasedLine.SpriteSize
   ) {module=CalculateLightSize}
  ]
  rendererClasses [
   &NiagaraMeshRendererProperties
   &NiagaraSpriteRendererProperties
   &NiagaraLightRendererProperties
  ]
  renderers [
   {rendererClass=&NiagaraMeshRendererProperties, rendererIndex=0}
   {rendererClass=&NiagaraSpriteRendererProperties, rendererIndex=1}
   {rendererClass=&NiagaraLightRendererProperties, rendererIndex=2}
  ]
 }
 SwarmPlayers {
  emitterSpawnScript [/SetVariables(Emitter.ThisEmittersArrangementsID:ENiagara_Papaya_FluorescentLightArrangements=::NewEnumerator3) {isSetParameters=true}]
  emitterUpdateScript [
   &EmitterState()
   &self:GiveBirthToParticlesIfItsTheirTime() {module=GiveBirthToParticlesIfItsTheirTime}
  ]
  particleSpawnScript [
   &InitializeParticle(
    Module.Write Color=true
    Module.Write Lifetime=true
    Module.Write Mass=true
    Module.Write Material Random=true
    Module.Write Position=false
    Module.Write Scale=false
    Module.Write SpriteRotation=false
    Module.Write SpriteSize=false
    Sprite Size=<50 50>
   )
   /SetVariables(
    Particles.RandomNumber:NiagaraFloat=&UniformRangedFloat(Minimum=0, Maximum=1)
    Particles.RandomNumber2:NiagaraFloat=&UniformRangedFloat(Minimum=0, Maximum=1)
    Particles.RandomNumber3:NiagaraFloat=&UniformRangedFloat(Minimum=-1, Maximum=1)
    Particles.RandomPlayerIndex:NiagaraInt32=&UniformRangedInt(
     Minimum=0
     Maximum=/Niagara/DynamicInputs/Subtract/Subtract_Integer(
      A:NiagaraInt32=/Niagara/DynamicInputs/Math/GetVectorArrayCount(Vector Array:NiagaraDataInterfaceArrayFloat3=$User.ArrayOfNearestPlayerPositions)
      B:NiagaraInt32=1
     )
    )
    Particles.Random_NegativeOrPositive1:NiagaraInt32=&MakeCustomIntFromBool(
     Bool=/Niagara/DynamicInputs/Random/RandomBool(Randomness Mode:ENiagaraRandomnessMode=::NewEnumerator0, Random Seed:NiagaraInt32=0, Fixed Random Seed:NiagaraBool=false)
     True INT=1
     False INT=-1
    )
   ) {isSetParameters=true}
   &self:ReadSenderOtherParticleAttributes() {module=ReadSenderOtherParticleAttributes}
   /SetVariables(Particles.PercussionLookupIndex:NiagaraFloat=&UniformRangedFloat(Minimum=0, Maximum=1)) {isSetParameters=true}
   /SetVariables(Particles.RandomEnergyBurstMultiplier:NiagaraFloat=&UniformRangedFloat(Minimum=0, Maximum=1)) {isSetParameters=true}
   /SetVariables(Particles.NegativeOrPositive1:NiagaraFloat=&UniformAOrBFloat(A=-1)) {isSetParameters=true}
  ]
  particleUpdateScript [
   &self:MultiLerpPercussionSamples(PercussionLookupIndex:NiagaraFloat=$Particles.PercussionLookupIndex) {module=MultiLerpPercussionSamples}
   &self:PickRotationCenterVector(
    FrameCounter:NiagaraInt32=&UniformRangedInt(Minimum=0, Maximum=50000)
    GroupID:NiagaraInt32=&UniformRangedInt(Minimum=0, Maximum=50000)
    Rate Of Change:NiagaraFloat=&Multiply_Float(A=$Engine.DeltaTime, B=1)
   ) {module=PickRotationCenterVector}
   &self:TemporalFade(
    Current Value:NiagaraFloat=&Multiply_Float(A=0.5, B=&Add_Float(A=$Particles.PercussionValue, B=$NPC.Papaya_Jukebox_NPC.AverageAmplitude))
    Fade Rate=3
   ) {module=TemporalFade}
   &ParticleState(Kill Particles When Lifetime Has Elapsed=false)
   &self:SeeIfParticlesNeedToDie(Number Of Frames To Linger:NiagaraInt32=$System.Number Of Frames To Linger) {module=SeeIfParticlesNeedToDie}
   &Drag(Drag=&Lerp_Float(A=0, B=1, Alpha=$Output.InfluencePercentage), Use Linear Drag=true, Use Rotational Drag=false)
   &self:GrabPlayerLocation(
    Is player valid:NiagaraDataInterfaceArrayInt32=$User.IsCharacterValid
    New Vector Array:NiagaraDataInterfaceArrayFloat3=$User.ArrayOfNearestPlayerPositions
    Target Player Index:NiagaraInt32=$Particles.RandomPlayerIndex
   ) {module=GrabPlayerLocation}
   &self:NormalizedDistanceFalloff(
    AttractorPosition:Vector3f=$Output.GrabPlayerLocation.PlayerLocation
    Behaviour Gradient Distance:NiagaraFloat=&MakeCustomFloatFromBool(Bool=$Output.GrabPlayerLocation.Following Player, True Float=256, False Float=1024)
    ParticlePosition:Vector3f=$Particles.Position_Top
    Start Distance Fade:NiagaraFloat=&MakeCustomFloatFromBool(Bool=$Output.GrabPlayerLocation.Following Player, True Float=10, False Float=1024)
   ) {module=NormalizedDistanceFalloff}
   /Niagara/Modules/Update/Utility/TemporalLerp_Float(
    Current Value:NiagaraFloat=&UniformRangedFloat(Minimum=0.5, Maximum=1.5)
    Rate Of Change:NiagaraFloat=&Multiply_Float(A=$Engine.DeltaTime, B=10)
   ) {enabled=false}
   &RotateAroundPoint(
    Derive the Rotation Axis=false
    Radius:NiagaraFloat=&Multiply_Float(
     A=&MakeCustomFloatFromBool(Bool=$Output.GrabPlayerLocation.Following Player, True Float=1, False Float=20)
     B=&Lerp_Float(A=150, B=30, Alpha=$Particles.RandomNumber)
    )
    Rotation Axis:Vector3f=&self:ScratchDynamicInput()
    Rotation Center:NiagaraPosition=&self:Find_Rotation_Center(
     Faded Value:NiagaraFloat=$Particles.TemporalFade.CurrentValue
     Following Player:NiagaraBool=$Output.GrabPlayerLocation.Following Player
     Negative or Positive 1:NiagaraFloat=$Particles.NegativeOrPositive1
     Player position:Vector3f=$Output.GrabPlayerLocation.PlayerLocation
     Player Total Height:NiagaraFloat=60
     Player up and Down Movement Amount:NiagaraFloat=1
     Random 0-1:NiagaraFloat=$Particles.RandomNumber2
     Tornado Total Height:NiagaraFloat=2000
     Twister Up and Down Movement Amount:NiagaraFloat=500
    )
    Rotation Phase:NiagaraFloat=$Particles.RandomNumber2
    Rotation Rate:NiagaraFloat=&Multiply_Float(
     A=$Particles.NegativeOrPositive1
     B=&Multiply_Float(
      A=&Multiply_Float(A=$Output.InfluencePercentage, B=&Lerp_Float(A=1, B=2, Alpha=$Particles.TemporalFade.CurrentValue))
      B=&MakeCustomFloatFromBool(Bool=$Output.GrabPlayerLocation.Following Player, True Float=1, False Float=0.125)
     )
    )
    Set Particle Position=false
    X Axis:Vector3f=&MakeVector(X=1, Y=0, Z=0)
   )
   &RotateAroundPoint(
    Derive the Rotation Axis=false
    Radius:NiagaraFloat=&Multiply_Float(
     A=&MakeCustomFloatFromBool(Bool=$Output.GrabPlayerLocation.Following Player, True Float=1, False Float=20)
     B=&Lerp_Float(A=150, B=30, Alpha=$Particles.RandomNumber)
    )
    Rotation Axis:Vector3f=&self:ScratchDynamicInput()
    Rotation Center:NiagaraPosition=&self:Find_Rotation_Center(
     Faded Value:NiagaraFloat=$Particles.TemporalFade.CurrentValue
     Following Player:NiagaraBool=$Output.GrabPlayerLocation.Following Player
     Negative or Positive 1:NiagaraFloat=$Particles.NegativeOrPositive1
     Player position:Vector3f=$Output.GrabPlayerLocation.PlayerLocation
     Player Total Height:NiagaraFloat=60
     Player up and Down Movement Amount:NiagaraFloat=1
     Random 0-1:NiagaraFloat=$Particles.RandomNumber2
     Tornado Total Height:NiagaraFloat=2000
     Twister Up and Down Movement Amount:NiagaraFloat=500
    )
    Rotation Phase:NiagaraFloat=$Particles.RandomNumber2
    Rotation Rate:NiagaraFloat=&Multiply_Float(
     A=$Particles.NegativeOrPositive1
     B=&Multiply_Float(
      A=&Multiply_Float(A=$Output.InfluencePercentage, B=&Lerp_Float(A=1, B=2, Alpha=$Particles.TemporalFade.CurrentValue))
      B=&MakeCustomFloatFromBool(Bool=$Output.GrabPlayerLocation.Following Player, True Float=1, False Float=0.125)
     )
    )
    Set Particle Position=false
    X Axis:Vector3f=&MakeVector(X=1, Y=0, Z=0)
   ) {enabled=false, module=RotateAroundPoint001}
   &SpringForce(
    " Particle Equilibrium Position":NiagaraPosition=$Output.RotateAroundPoint.OutputPosition
    Dampening Coefficient=10
    Desired Minimum Separation Distance=0
    Particle Position:Vector3f=$Particles.Position_Top
    Particle Velocity:Vector3f=$Particles.Velocity_Top
    Spring Tightness:NiagaraFloat=&self:Player_Swarm_Spring_Tightness(
     Following Player:NiagaraBool=$Output.GrabPlayerLocation.Following Player
     Player AKA Center Position:Vector3f=$Output.GrabPlayerLocation.PlayerLocation
     Top Particle position:Vector3f=$Particles.Position_Top
     Far Player Tightness:NiagaraFloat=200
     Near Player Tightness:NiagaraFloat=100
     Player Fade Distance:NiagaraFloat=256
     Player Min Distance:NiagaraFloat=64
     Far Distance Tightness Tornado:NiagaraFloat=200
     Near Distance Tightness Tornado:NiagaraFloat=100
     Tornado Fade Distance:NiagaraFloat=1024
     Tornado Min Distance:NiagaraFloat=512
    )
    Write to Intrinsic Parameters=false
   )
   &SpringForce(
    " Particle Equilibrium Position":Vector3f=$Particles.Position_Top
    Attachment Point Velocity:Vector3f=$Particles.Velocity_Top
    Dampening Coefficient=10
    Desired Minimum Separation Distance=0
    Particle Position:Vector3f=$Particles.Position_Bottom
    Particle Velocity:Vector3f=$Particles.Velocity_Bottom
    Spring Tightness:NiagaraFloat=&self:Player_Swarm_Spring_Tightness(
     Following Player:NiagaraBool=$Output.GrabPlayerLocation.Following Player
     Player AKA Center Position:Vector3f=$Output.GrabPlayerLocation.PlayerLocation
     Top Particle position:Vector3f=$Particles.Position_Top
     Far Player Tightness:NiagaraFloat=20000
     Near Player Tightness:NiagaraFloat=61.015
     Player Fade Distance:NiagaraFloat=256
     Player Min Distance:NiagaraFloat=125.0322
     Far Distance Tightness Tornado:NiagaraFloat=300
     Near Distance Tightness Tornado:NiagaraFloat=0.5
     Tornado Fade Distance:NiagaraFloat=1024
     Tornado Min Distance:NiagaraFloat=512
    )
    Write to Intrinsic Parameters=false
   ) {module=SpringForce001}
   /Niagara/Modules/Update/Forces/PointAttractionForce(
    Attraction Radius=109.4102
    AttractionStrength=-120
    AttractorPosition:Vector3f=$Output.GrabPlayerLocation.PlayerLocation
    Falloff Exponent=1
    Module.Kill Within Radius=false
    Module.Use Falloff=true
   ) {enabled=false, module=PointAttractionForce001}
   &self:Set_Top_Or_Bottom_Force(Bottom Force:Vector3f=$Output.SpringForce001.SpringForce, Top Force:Vector3f=$Output.SpringForce.SpringForce) {module=Set_Top_Or_Bottom_Force}
   &self:RotateVelocitiesAwayFromThePlayer(PlayerPosition:Vector3f=$Output.GrabPlayerLocation.PlayerLocation, Velocity Reflection Distance=100) {enabled=false, module=RotateVelocitiesAwayFromThePlayer}
   &self:-DUPLICATE_BELOW_THIS_MODULE-__Names() {module=-DUPLICATE_BELOW_THIS_MODULE-__Names}
   &self:BlendInInheritedVelocityAtSpawn(Clamped Normalized Age Curve:NiagaraDataInterfaceCurve=$System.Transfer to New Velocity Curve) {module=BlendInInheritedVelocityAtSpawn}
   &SolveForcesAndVelocity(
    Acceleration Limit:NiagaraFloat=$System.AccelerationLimit
    Clamp Velocity=true
    Force:Vector3f=$Transient.Top_PhysicsForce
    Limit Acceleration=true
    Position:Vector3f=$Particles.Position_Top
    PreviousVelocity:Vector3f=$Particles.Previous.Velocity_Top
    Speed Limit:NiagaraFloat=$System.MaxSpeed
    Velocity:Vector3f=$Particles.Velocity_Top
    Write to Intrinsic Properties=false
   )
   &SolveForcesAndVelocity(
    Acceleration Limit:NiagaraFloat=$System.AccelerationLimit
    Clamp Velocity=true
    Force:Vector3f=$Transient.Bottom_PhysicsForce
    Limit Acceleration=true
    Position:Vector3f=$Particles.Position_Bottom
    PreviousVelocity:Vector3f=$Particles.Previous.Velocity_Bottom
    Speed Limit:NiagaraFloat=$System.MaxSpeed
    Velocity:Vector3f=$Particles.Velocity_Bottom
    Write to Intrinsic Properties=false
   ) {module=SolveForcesAndVelocity001}
   &self:StoreOffTopAndBottomVariables(
    Bottom Position:NiagaraPosition=$Output.SolveForcesAndVelocity001.Position
    Bottom Velocity:Vector3f=$Output.SolveForcesAndVelocity001.Velocity
    Top Position:NiagaraPosition=$Output.SolveForcesAndVelocity.Position
    Top Velocity:Vector3f=$Output.SolveForcesAndVelocity.Velocity
   ) {module=StoreOffTopAndBottomVariables}
   &self:InitialTransferAnimation_(
    Include Initialization Rotation Animation=false
    RandomAxisOfRotation:Vector3f=$Particles.PickRotationCenterVector.RotationAxis
    Rotation Modulation:NiagaraFloat=$Particles.RandomEnergyBurstMultiplier
   ) {module=InitialTransferAnimation_}
   &self:FeedPositionsIntoShader(
    Bottom Position:Vector3f=$Particles.Position_Bottom
    Initial PercussionValue:NiagaraFloat=$Particles.Initial.PercussionValue
    PercussionValue:NiagaraFloat=$Particles.PercussionValue
    Top Position:Vector3f=$Particles.Position_Top
   ) {module=FeedPositionsIntoShader}
   &SpriteBasedLine(Point A:Vector3f=$Particles.Position_Top, Point B:Vector3f=$Particles.Position_Bottom, SpriteWidth=0.01)
   &CalculateAccurateVelocity(
    Current Position:Vector3f=$Particles.Position_Top
    Previous Position:Vector3f=$Particles.Previous.Position_Top
    Set Intrinsic Variables=false
   )
   &CalculateAccurateVelocity(
    Current Position:Vector3f=$Particles.Position_Bottom
    Previous Position:Vector3f=$Particles.Previous.Position_Bottom
    Set Intrinsic Variables=false
   ) {module=CalculateAccurateVelocity001}
   /SetVariables(
    Particles.Position=$Particles.SpriteBasedLine.Position
    Particles.Velocity_Bottom:Vector3f=$Output.CalculateAccurateVelocity001.Velocity
    Particles.Velocity_Top:Vector3f=$Output.CalculateAccurateVelocity.Velocity
   ) {isSetParameters=true}
   &self:CalculateLightSize(
    Color:LinearColor=&MakeLinearColorFromVectorAndFloat(
     "Vector (RGB)"=&MakeVectorFromLinearColorRGB(Color=&MultiplyLinearColor(
       Linear Color=&ColorFromCurve(ColorCurve=$System.BaseColorCurve, CurveIndex=&Frac_Float(Float=$NPC.Papaya_Jukebox_NPC.Base Color Parameter))
       Scale Factor=&MakeVector4_FromFloat(Float=&Lerp_Float(A=0.25, B=2, Alpha=$Particles.PercussionValue))
      ))
     "Float (Alpha)"=1
    )
    Sprite Based Line Size:Vector2f=$Particles.SpriteBasedLine.SpriteSize
   ) {module=CalculateLightSize}
   /SetVariables(Particles.LightColor=&MakeLinearColorFromVectorAndFloat(
     "Vector (RGB)"=&Multiply_VectorByFloat(
      Vector=&MakeVectorFromLinearColorRGB(Color=&MultiplyLinearColor(
        Linear Color=&ColorFromCurve(ColorCurve=$System.BaseColorCurve, CurveIndex=&Frac_Float(Float=$NPC.Papaya_Jukebox_NPC.Base Color Parameter))
        Scale Factor=&MakeVector4_FromFloat(Float=&Lerp_Float(A=1, B=1, Alpha=$Particles.TemporalFade.CurrentValue))
       ))
      Float=4200
     )
     "Float (Alpha)"=1
    )) {enabled=false, isSetParameters=true}
  ]
  rendererClasses [
   &NiagaraMeshRendererProperties
   &NiagaraSpriteRendererProperties
   &NiagaraLightRendererProperties
  ]
  renderers [
   {rendererClass=&NiagaraMeshRendererProperties, rendererIndex=0}
   {rendererClass=&NiagaraSpriteRendererProperties, rendererIndex=1}
   {rendererClass=&NiagaraLightRendererProperties, rendererIndex=2}
  ]
 }
 Tetrahedra {
  emitterSpawnScript [
   /SetVariables(Emitter.ThisEmittersArrangementsID:ENiagara_Papaya_FluorescentLightArrangements=::NewEnumerator4) {isSetParameters=true, module=SetVariables_#12}
   /SetVariables(Emitter.ShapePointArray=DataInterface() {propertyValues {GpuSyncMode=SyncCpuToGpu, MaxElements=0, FloatData [<-0.5 0.5 0.5>, <0.5 -0.5 0.5>, <-0.5 -0.5 -0.5>, <0.5 0.5 -0.5>]}}) {isSetParameters=true, module=SetVariables_#16}
   /SetVariables(Emitter.Edges=DataInterface() {propertyValues {GpuSyncMode=SyncCpuToGpu, MaxElements=0, FloatData [<0 1>, <0 2>, <1 2>, <0 3>, <1 3>, <2 3>]}}) {isSetParameters=true, module=SetVariables_#17}
   /SetVariables(Emitter.RatioOfDistanceBetweenTetrehedrons=30) {isSetParameters=true, module=SetVariables_#21}
  ]
  emitterUpdateScript [
   &EmitterState()
   &self:GiveBirthToParticlesIfItsTheirTime() {module=GiveBirthToParticlesIfItsTheirTime}
   &SpawnBurst_Instantaneous(Spawn Count=55, Use Spawn Probability=false) {enabled=false}
  ]
  particleSpawnScript [
   &InitializeParticle(
    Module.Write Color=true
    Module.Write Lifetime=true
    Module.Write Mass=true
    Module.Write Material Random=true
    Module.Write Position=true
    Module.Write Scale=false
    Module.Write SpriteRotation=false
    Module.Write SpriteSize=false
    Position=<0 0 0>
    Sprite Size=<50 50>
   )
   &self:CalculateLocalTetrahedronPositions(
    New Vector 2D Array:NiagaraDataInterfaceArrayFloat2=$Emitter.Edges
    New Vector Array:NiagaraDataInterfaceArrayFloat3=$Emitter.ShapePointArray
   ) {module=CalculateLocalTetrahedronPositions}
   /Niagara/Modules/Spawn/Location/V2/GridLocation(
    Align To Grid Cell=true
    Coordinate Space:ENiagaraCoordinateSpace=::World
    Dimensions Definition:ENiagaraGridPlacementType=::NewEnumerator1
    Exec Index:NiagaraInt32=$Particles.GroupID
    Grid Origin:NiagaraPosition=<0 0 0>
    Grid X Count=3
    Grid Y Count=3
    Grid Z Count=1
    Module.Jitter Inside Grid Cells=true
    Module.Override Random Seed=true
    Offset:Vector3f=<0.5 0.5 0>
    Placement Randomness Mode:ENiagaraRandomnessMode=::NewEnumerator1
    Random Placement Seed:NiagaraInt32=&Add_Integer(A=$Particles.GroupID, B=$Engine.System.TickCount)
    Randomize Overriden Seed=false
    Randomize Placement Within Cell:Vector3f=<0 0 0>
    XYZ Dimensions:Vector3f=&MakeVector(X=1, Y=1, Z=0)
   )
   &self:Initial_Tetrahedron_Macro_Transform(
    Position Center:Vector3f=$User.PolyhedraCenter
    X Vector:Vector3f=$User.PolyhedraXAxis
    Y Vector:Vector3f=$User.PolyhedraYAxis
    groupID:NiagaraInt32=$Particles.GroupID
   ) {module=Initial_Tetrahedron_Macro_Transform}
   &self:RotateTetrahedron(
    RotationAmount:NiagaraFloat=&UniformRangedFloat(Minimum=0, Maximum=1, Randomness Mode=::NewEnumerator1, Random Seed=$Particles.GroupID, Fixed Random Seed=true)
    RotationAxis:Vector3f=/Niagara/DynamicInputs/Random/RandomVector(
     Vector Scale:NiagaraFloat=1
     Constraint Vector:Vector3f=<0 0 1>
     Randomness Mode:ENiagaraRandomnessMode=::NewEnumerator1
     Random Seed:NiagaraInt32=$Particles.GroupID
     Fixed Random Seed:NiagaraBool=true
    )
   ) {module=RotateTetrahedron}
   &self:Find_final_Tetrahedron_Render_Positions(Scale:NiagaraFloat=&Multiply_Float(A=200, B=$Particles.Initial Scale Factor)) {module=Find_final_Tetrahedron_Render_Positions}
   /SetVariables(Particles.Bottom_TargetLocation:Vector3f=$Particles.Bottom_TetrahedronPoint) {isSetParameters=true, module=SetVariables_#18}
   /SetVariables(Particles.Top_TargetLocation:Vector3f=$Particles.Top_TetrahedronPoint) {isSetParameters=true, module=SetVariables_#19}
   &self:ReadSenderOtherParticleAttributes() {module=ReadSenderOtherParticleAttributes}
   /SetVariables(Particles.PercussionLookupIndex:NiagaraFloat=&UniformRangedFloat(
     Minimum=0
     Maximum=1
     Randomness Mode=::NewEnumerator1
     Random Seed=&Add_Integer(A=$Particles.GroupID, B=$Engine.System.TickCount)
     Fixed Random Seed=true
    )) {isSetParameters=true, module=SetVariables_#13}
   /SetVariables(Particles.PercussionLookupIndex:NiagaraFloat=&VectorLength(Vector=&Subtract_Vector(A=$Output.GridLocation.GridUVW, B=<0.5 0.5 0>))) {enabled=false, isSetParameters=true}
   /SetVariables(Particles.RandomEnergyBurstMultiplier:NiagaraFloat=&UniformRangedFloat(Minimum=0, Maximum=1)) {isSetParameters=true, module=SetVariables_#14}
   /SetVariables() {isSetParameters=true}
  ]
  particleUpdateScript [
   &self:MultiLerpPercussionSamples(PercussionLookupIndex:NiagaraFloat=$Particles.PercussionLookupIndex) {module=MultiLerpPercussionSamples}
   &self:PickRotationCenterVector(
    FrameCounter:NiagaraInt32=&UniformRangedInt(Minimum=0, Maximum=50000)
    GroupID:NiagaraInt32=&UniformRangedInt(Minimum=0, Maximum=50000)
    Rate Of Change:NiagaraFloat=&Multiply_Float(A=$Engine.DeltaTime, B=1)
   ) {module=PickRotationCenterVector001}
   &self:TemporalFade(
    Current Value:NiagaraFloat=&Multiply_Float(A=0.5, B=&Add_Float(A=$Particles.PercussionValue, B=$NPC.Papaya_Jukebox_NPC.AverageAmplitude))
    Fade Rate=2
   ) {module=TemporalFade}
   &ParticleState(Kill Particles When Lifetime Has Elapsed=false)
   &CurlNoiseForce(
    Noise Frequency=0.0002
    Noise Strength=&Lerp_Float(A=0, B=713.3311, Alpha=$Particles.TemporalFade.CurrentValue)
    Pan Noise=true
    Pan Noise Field=<0.1 0.1 0.1>
    Sample Position:Vector3f=$Particles.GroupCenter
    Write to Intrinsic Parameters=false
   )
   &SpringForce(
    " Particle Equilibrium Position":Vector3f=$Particles.Initial.GroupCenter
    Desired Minimum Separation Distance=0
    Particle Position:Vector3f=$Particles.GroupCenter
    Particle Velocity:Vector3f=$Particles.GroupCenterVelocity
    Spring Tightness=0.6011
    Write to Intrinsic Parameters=false
   ) {module=SpringForce002}
   &SolveForcesAndVelocity(
    Clamp Velocity=false
    Force:Vector3f=&Add_Vector(A=$Output.SpringForce002.SpringForce, B=$Output.CurlNoiseForce.CurlNoiseForce)
    Limit Acceleration=false
    Position:Vector3f=$Particles.GroupCenter
    PreviousVelocity:Vector3f=$Particles.Previous.GroupCenterVelocity
    Velocity:Vector3f=$Particles.GroupCenterVelocity
    Write to Intrinsic Properties=false
   ) {enabled=false, module=SolveForcesAndVelocity002}
   &SolveForcesAndVelocity(
    Clamp Velocity=false
    Force:Vector3f=&Add_Vector(A=$Output.SpringForce002.SpringForce, B=$Output.CurlNoiseForce.CurlNoiseForce)
    Limit Acceleration=false
    Position:Vector3f=$Particles.GroupCenter
    Velocity:Vector3f=$Particles.GroupCenterVelocity
    Write to Intrinsic Properties=false
   ) {module=SolveForcesAndVelocity003}
   /SetVariables(Particles.GroupCenter:NiagaraPosition=$Output.SolveForcesAndVelocity003.Position) {isSetParameters=true}
   /SetVariables(Particles.GroupCenterVelocity_ManualPrevious:Vector3f=$Particles.GroupCenterVelocity) {isSetParameters=true}
   &CalculateAccurateVelocity(Current Position:Vector3f=$Particles.GroupCenter, Previous Position:Vector3f=$Particles.Previous.GroupCenter, Set Intrinsic Variables=false) {module=CalculateAccurateVelocity004}
   /SetVariables(Particles.GroupCenterVelocity:Vector3f=$Output.CalculateAccurateVelocity004.Velocity) {isSetParameters=true}
   &self:SeeIfParticlesNeedToDie(Number Of Frames To Linger:NiagaraInt32=$System.Number Of Frames To Linger) {module=SeeIfParticlesNeedToDie}
   &self:PickRotationCenterVector(
    FrameCounter:NiagaraInt32=$Emitter.EmitterFrameCounter.Counter
    GroupID:NiagaraInt32=$Particles.GroupID
    Rate Of Change:NiagaraFloat=&Multiply_Float(A=$Engine.DeltaTime, B=1)
   ) {module=PickRotationCenterVector}
   &self:RotateTetrahedron(
    RotationAmount:NiagaraFloat=&Multiply_Float(A=$Engine.DeltaTime, B=&Lerp_Float(A=0.125, B=3, Alpha=$Particles.TemporalFade.CurrentValue))
    RotationAxis:Vector3f=$Particles.PickRotationCenterVector.RotationAxis
   ) {module=RotateTetrahedron001}
   &self:RotateLocalUnitSphere(RotationAxis:Vector3f=$Particles.PickRotationCenterVector.RotationAxis) {enabled=false, module=RotateLocalUnitSphere002}
   &self:FindSphereTargetPositions(BottomDistance=100, Group Center:Vector3f=$Particles.GroupCenter, SphereUnitVector:Vector3f=$Particles.LocalUnitSphere, TopDistance=200) {enabled=false, module=FindSphereTargetPositions001}
   &self:Find_final_Tetrahedron_Render_Positions(Scale:NiagaraFloat=&Multiply_Float(
     A=200
     B=&Multiply_Float(A=$Particles.Initial Scale Factor, B=&Lerp_Float(A=1, B=1.25, Alpha=$Particles.TemporalFade.CurrentValue))
    )) {module=Find_final_Tetrahedron_Render_Positions001}
   &CalculateAccurateVelocity(
    Current Position:Vector3f=$Particles.Bottom_TetrahedronPoint
    Previous Position:Vector3f=$Particles.Previous.Bottom_TetrahedronPoint
    Set Intrinsic Variables=false
   )
   &CalculateAccurateVelocity(
    Current Position:Vector3f=$Particles.Top_TetrahedronPoint
    Previous Position:Vector3f=$Particles.Previous.Top_TetrahedronPoint
    Set Intrinsic Variables=false
   ) {module=CalculateAccurateVelocity001}
   &SpringForce(
    " Particle Equilibrium Position":Vector3f=$Particles.Top_TetrahedronPoint
    Attachment Point Velocity:Vector3f=$Output.CalculateAccurateVelocity001.Velocity
    Dampening Coefficient=10.9034
    Desired Minimum Separation Distance=0
    Particle Position:Vector3f=$Particles.Position_Top
    Particle Velocity:Vector3f=$Particles.Velocity_Top
    Spring Tightness:NiagaraFloat=&Multiply_Float(A=600, B=&ClampFloat(Float=&Multiply_Float(A=$Particles.Age, B=0.5)))
    Write to Intrinsic Parameters=false
   )
   &SpringForce(
    " Particle Equilibrium Position":Vector3f=$Particles.Bottom_TetrahedronPoint
    Attachment Point Velocity:Vector3f=$Output.CalculateAccurateVelocity.Velocity
    Dampening Coefficient=7.2428
    Desired Minimum Separation Distance=0
    Particle Position:Vector3f=$Particles.Position_Bottom
    Particle Velocity:Vector3f=$Particles.Velocity_Bottom
    Spring Tightness:NiagaraFloat=&Multiply_Float(A=600, B=&ClampFloat(Float=&Multiply_Float(A=$Particles.Age, B=0.5)))
    Write to Intrinsic Parameters=false
   ) {module=SpringForce001}
   &self:Set_Top_Or_Bottom_Force(Bottom Force:Vector3f=$Output.SpringForce001.SpringForce, Top Force:Vector3f=$Output.SpringForce.SpringForce) {module=Set_Top_Or_Bottom_Force}
   &self:-DUPLICATE_BELOW_THIS_MODULE-__Names() {module=-DUPLICATE_BELOW_THIS_MODULE-__Names}
   &self:BlendInInheritedVelocityAtSpawn(Clamped Normalized Age Curve:NiagaraDataInterfaceCurve=$System.Transfer to New Velocity Curve) {module=BlendInInheritedVelocityAtSpawn}
   &SolveForcesAndVelocity(
    Acceleration Limit:NiagaraFloat=$System.AccelerationLimit
    Clamp Velocity=true
    Force:Vector3f=$Transient.Top_PhysicsForce
    Limit Acceleration=true
    Position:Vector3f=$Particles.Position_Top
    PreviousVelocity:Vector3f=$Particles.Previous.Velocity_Top
    Speed Limit:NiagaraFloat=$System.MaxSpeed
    Velocity:Vector3f=$Particles.Velocity_Top
    Write to Intrinsic Properties=false
   )
   &SolveForcesAndVelocity(
    Acceleration Limit:NiagaraFloat=$System.AccelerationLimit
    Clamp Velocity=true
    Force:Vector3f=$Transient.Bottom_PhysicsForce
    Limit Acceleration=true
    Position:Vector3f=$Particles.Position_Bottom
    PreviousVelocity:Vector3f=$Particles.Previous.Velocity_Bottom
    Speed Limit:NiagaraFloat=$System.MaxSpeed
    Velocity:Vector3f=$Particles.Velocity_Bottom
    Write to Intrinsic Properties=false
   ) {module=SolveForcesAndVelocity001}
   &self:StoreOffTopAndBottomVariables(
    Bottom Position:NiagaraPosition=$Output.SolveForcesAndVelocity001.Position
    Bottom Velocity:Vector3f=$Output.SolveForcesAndVelocity001.Velocity
    Top Position:NiagaraPosition=$Output.SolveForcesAndVelocity.Position
    Top Velocity:Vector3f=$Output.SolveForcesAndVelocity.Velocity
   ) {module=StoreOffTopAndBottomVariables}
   &self:InitialTransferAnimation_(
    Include Initialization Rotation Animation=false
    RandomAxisOfRotation:Vector3f=$Particles.PickRotationCenterVector001.RotationAxis
    Rotation Modulation:NiagaraFloat=$Particles.RandomEnergyBurstMultiplier
   ) {module=InitialTransferAnimation_}
   &self:FeedPositionsIntoShader(
    Bottom Position:Vector3f=$Particles.Position_Bottom
    Initial PercussionValue:NiagaraFloat=$Particles.Initial.PercussionValue
    PercussionValue:NiagaraFloat=$Particles.TemporalFade.CurrentValue
    Top Position:Vector3f=$Particles.Position_Top
   ) {module=FeedPositionsIntoShader}
   &self:FeedPositionsIntoShader(
    Bottom Position:Vector3f=$Particles.Bottom_TargetLocation
    Initial PercussionValue:NiagaraFloat=$Particles.Initial.PercussionValue
    PercussionValue:NiagaraFloat=$Particles.PercussionValue
    Top Position:Vector3f=$Particles.Top_TargetLocation
   ) {enabled=false, module=FeedPositionsIntoShader001}
   &SpriteBasedLine(Point A:Vector3f=$Particles.Position_Top, Point B:Vector3f=$Particles.Position_Bottom, SpriteWidth=0.01)
   &CalculateAccurateVelocity(
    Current Position:Vector3f=$Particles.Position_Top
    Previous Position:Vector3f=$Particles.Previous.Position_Top
    Set Intrinsic Variables=false
   ) {module=CalculateAccurateVelocity002}
   &CalculateAccurateVelocity(
    Current Position:Vector3f=$Particles.Position_Bottom
    Previous Position:Vector3f=$Particles.Previous.Position_Bottom
    Set Intrinsic Variables=false
   ) {module=CalculateAccurateVelocity003}
   /SetVariables(
    Particles.Position=$Particles.SpriteBasedLine.Position
    Particles.Velocity_Bottom:Vector3f=$Output.CalculateAccurateVelocity003.Velocity
    Particles.Velocity_Top:Vector3f=$Output.CalculateAccurateVelocity002.Velocity
   ) {isSetParameters=true, module=SetVariables_#15}
   &self:CalculateLightSize(
    Color:LinearColor=&MakeLinearColorFromVectorAndFloat(
     "Vector (RGB)"=&MakeVectorFromLinearColorRGB(Color=&MultiplyLinearColor(
       Linear Color=&ColorFromCurve(ColorCurve=$System.BaseColorCurve, CurveIndex=&Frac_Float(Float=$NPC.Papaya_Jukebox_NPC.Base Color Parameter))
       Scale Factor=&MakeVector4_FromFloat(Float=&Lerp_Float(A=0.5, B=2, Alpha=$Particles.TemporalFade.CurrentValue))
      ))
     "Float (Alpha)"=1
    )
    Sprite Based Line Size:Vector2f=$Particles.SpriteBasedLine.SpriteSize
   ) {module=CalculateLightSize}
  ]
  rendererClasses [
   &NiagaraMeshRendererProperties
   &NiagaraSpriteRendererProperties
   &NiagaraLightRendererProperties
  ]
  renderers [
   {rendererClass=&NiagaraMeshRendererProperties, rendererIndex=0}
   {rendererClass=&NiagaraSpriteRendererProperties, rendererIndex=1}
   {rendererClass=&NiagaraLightRendererProperties, rendererIndex=2}
  ]
 }
 Tetrahedra001 {
  bEnabled=false
  emitterSpawnScript [
   /SetVariables(Emitter.ThisEmittersArrangementsID:ENiagara_Papaya_FluorescentLightArrangements=::NewEnumerator4) {isSetParameters=true, module=SetVariables_#12}
   /SetVariables(Emitter.ShapePointArray=DataInterface() {propertyValues {GpuSyncMode=SyncCpuToGpu, MaxElements=0, FloatData [<-0.5 0.5 0.5>, <0.5 -0.5 0.5>, <-0.5 -0.5 -0.5>, <0.5 0.5 -0.5>]}}) {isSetParameters=true, module=SetVariables_#16}
   /SetVariables(Emitter.Edges=DataInterface() {propertyValues {GpuSyncMode=SyncCpuToGpu, MaxElements=0, FloatData [<0 1>, <0 2>, <1 2>, <0 3>, <1 3>, <2 3>]}}) {isSetParameters=true, module=SetVariables_#17}
   /SetVariables(Emitter.RatioOfDistanceBetweenTetrehedrons=30) {isSetParameters=true, module=SetVariables_#21}
  ]
  emitterUpdateScript [
   &EmitterState()
   &self:GiveBirthToParticlesIfItsTheirTime() {module=GiveBirthToParticlesIfItsTheirTime}
   &SpawnBurst_Instantaneous(Spawn Count=55, Use Spawn Probability=false)
   &EmitterFrameCounter() {enabled=false}
  ]
  particleSpawnScript [
   &InitializeParticle(
    Lifetime=9999
    Module.Write Color=true
    Module.Write Lifetime=true
    Module.Write Mass=true
    Module.Write Material Random=true
    Module.Write Position=true
    Module.Write Scale=false
    Module.Write SpriteRotation=false
    Module.Write SpriteSize=false
    Position=<0 0 0>
    Sprite Size=<50 50>
   )
   &self:CalculateLocalTetrahedronPositions(
    New Vector 2D Array:NiagaraDataInterfaceArrayFloat2=$Emitter.Edges
    New Vector Array:NiagaraDataInterfaceArrayFloat3=$Emitter.ShapePointArray
   ) {module=CalculateLocalTetrahedronPositions}
   &GridLocation(
    Align To Grid Cell=true
    Coordinate Space:ENiagaraCoordinateSpace=::World
    Dimensions Definition:ENiagaraGridPlacementType=::NewEnumerator1
    Exec Index:NiagaraInt32=$Particles.GroupID
    Grid Origin:NiagaraPosition=<0 0 0>
    Grid X Count=3
    Grid Y Count=3
    Grid Z Count=1
    Module.Jitter Inside Grid Cells=true
    Module.Override Random Seed=true
    Offset:Vector3f=<0.5 0.5 0>
    Placement Randomness Mode:ENiagaraRandomnessMode=::NewEnumerator1
    Random Placement Seed:NiagaraInt32=&Add_Integer(A=$Particles.GroupID, B=$Engine.System.TickCount)
    Randomize Overriden Seed=false
    Randomize Placement Within Cell:Vector3f=<0 0 0>
    XYZ Dimensions:Vector3f=&MakeVector(X=1, Y=1, Z=0)
   )
   &self:Initial_Tetrahedron_Macro_Transform(
    Position Center:Vector3f=$User.PolyhedraCenter
    X Vector:Vector3f=$User.PolyhedraXAxis
    Y Vector:Vector3f=$User.PolyhedraYAxis
    groupID:NiagaraInt32=$Particles.GroupID
   ) {module=Initial_Tetrahedron_Macro_Transform}
   &self:RotateTetrahedron(
    RotationAmount:NiagaraFloat=&UniformRangedFloat(Minimum=0, Maximum=1, Randomness Mode=::NewEnumerator1, Random Seed=$Particles.GroupID, Fixed Random Seed=true)
    RotationAxis:Vector3f=&RandomVector(
     Vector Scale:NiagaraFloat=1
     Constraint Vector:Vector3f=<0 0 1>
     Randomness Mode:ENiagaraRandomnessMode=::NewEnumerator1
     Random Seed:NiagaraInt32=$Particles.GroupID
     Fixed Random Seed:NiagaraBool=true
    )
   ) {module=RotateTetrahedron}
   &self:Find_final_Tetrahedron_Render_Positions(Scale:NiagaraFloat=&Multiply_Float(A=200, B=$Particles.Initial Scale Factor)) {module=Find_final_Tetrahedron_Render_Positions}
   /SetVariables(Particles.Bottom_TargetLocation:Vector3f=$Particles.Bottom_TetrahedronPoint) {isSetParameters=true, module=SetVariables_#18}
   /SetVariables(Particles.Top_TargetLocation:Vector3f=$Particles.Top_TetrahedronPoint) {isSetParameters=true, module=SetVariables_#19}
   &self:FeedPositionsIntoShader(
    Bottom Position:Vector3f=$Particles.Bottom_TargetLocation
    Initial PercussionValue:NiagaraFloat=$Particles.Initial.PercussionValue
    PercussionValue:NiagaraFloat=$Particles.PercussionValue
    Top Position:Vector3f=$Particles.Top_TargetLocation
   ) {module=FeedPositionsIntoShader}
   &self:ReadSenderOtherParticleAttributes() {enabled=false, module=ReadSenderOtherParticleAttributes}
   /SetVariables(Particles.PercussionLookupIndex:NiagaraFloat=&UniformRangedFloat(
     Minimum=0
     Maximum=1
     Randomness Mode=::NewEnumerator1
     Random Seed=&Add_Integer(A=$Particles.GroupID, B=$Engine.System.TickCount)
     Fixed Random Seed=true
    )) {enabled=false, isSetParameters=true, module=SetVariables_#13}
  ]
  particleUpdateScript [
   &ParticleState()
   &self:FeedPositionsIntoShader(
    Bottom Position:Vector3f=$Particles.Bottom_TargetLocation
    Initial PercussionValue:NiagaraFloat=$Particles.Initial.PercussionValue
    PercussionValue:NiagaraFloat=$Particles.PercussionValue
    Top Position:Vector3f=$Particles.Top_TargetLocation
   ) {enabled=false, module=FeedPositionsIntoShader001}
   /SetVariables(
    Particles.Position:Vector3f=$Particles.SpriteBasedLine.Position
    Particles.Velocity_Bottom:Vector3f=$Output.CalculateAccurateVelocity003.Velocity
    Particles.Velocity_Top:Vector3f=$Output.CalculateAccurateVelocity002.Velocity
   ) {enabled=false, isSetParameters=true, module=SetVariables_#15}
   &self:CalculateLightSize(Sprite Based Line Size:Vector2f=$Particles.SpriteBasedLine.SpriteSize) {enabled=false, module=CalculateLightSize}
  ]
  rendererClasses [
   &NiagaraMeshRendererProperties
   &NiagaraSpriteRendererProperties
   &NiagaraLightRendererProperties
  ]
  renderers [
   {rendererClass=&NiagaraMeshRendererProperties, rendererIndex=0}
   {rendererClass=&NiagaraSpriteRendererProperties, rendererIndex=1}
   {rendererClass=&NiagaraLightRendererProperties, rendererIndex=2}
  ]
 }
}
summary {
 systemName=P_Papaya_FluorescentLights
 emitters [
  {bEnabled=false, emitterName=RecursiveLightEmitter, rendererClasses [&NiagaraMeshRendererProperties]}
  {
   emitterName=SpawnLine
   rendererClasses [
    &NiagaraSpriteRendererProperties
    &NiagaraMeshRendererProperties
    &NiagaraLightRendererProperties
   ]
  }
  {
   emitterName=Dancers
   rendererClasses [
    &NiagaraMeshRendererProperties
    &NiagaraSpriteRendererProperties
    &NiagaraLightRendererProperties
   ]
  }
  {
   emitterName=Spheres
   rendererClasses [
    &NiagaraMeshRendererProperties
    &NiagaraSpriteRendererProperties
    &NiagaraLightRendererProperties
   ]
  }
  {
   emitterName=SwarmPlayers
   rendererClasses [
    &NiagaraMeshRendererProperties
    &NiagaraSpriteRendererProperties
    &NiagaraLightRendererProperties
   ]
  }
  {
   bEnabled=false
   emitterName=Polyhedra
   rendererClasses [
    &NiagaraMeshRendererProperties
    &NiagaraSpriteRendererProperties
    &NiagaraLightRendererProperties
   ]
  }
  {
   emitterName=Tetrahedra
   rendererClasses [
    &NiagaraMeshRendererProperties
    &NiagaraSpriteRendererProperties
    &NiagaraLightRendererProperties
   ]
  }
  {
   bEnabled=false
   emitterName=Tetrahedra001
   rendererClasses [
    &NiagaraMeshRendererProperties
    &NiagaraSpriteRendererProperties
    &NiagaraLightRendererProperties
   ]
  }
  {
   emitterName=Die
   rendererClasses [
    &NiagaraMeshRendererProperties
    &NiagaraSpriteRendererProperties
    &NiagaraLightRendererProperties
   ]
  }
  {bEnabled=false, emitterName=SimpleSpriteBurst, rendererClasses [&NiagaraSpriteRendererProperties]}
 ]
 userVariables [
  User.ActivationCount {type=NiagaraInt32}
  User.ArrayOfNearestPlayerPositions {
   type=NiagaraDataInterfaceArrayFloat3
   description=""
   defaultValue {
    struct=NiagaraExt_VariableValue_DataInterface
    value {
     dataInterface=&self:NiagaraDataInterfaceArrayFloat3_0
     dataInterfaceClass=&NiagaraDataInterfaceArrayFloat3
    }
   }
  }
  User.BaseColorCurve {
   type=NiagaraDataInterfaceColorCurve
   description=""
   defaultValue {
    struct=NiagaraExt_VariableValue_DataInterface
    value {
     dataInterface=&self:NiagaraDataInterfaceColorCurve_0
     dataInterfaceClass=&NiagaraDataInterfaceColorCurve
    }
   }
  }
  User.ExitLocations {
   type=NiagaraDataInterfaceArrayFloat3
   description=""
   defaultValue {
    struct=NiagaraExt_VariableValue_DataInterface
    value {
     dataInterface=&self:NiagaraDataInterfaceArrayFloat3_1
     dataInterfaceClass=&NiagaraDataInterfaceArrayFloat3
    }
   }
  }
  User.IsCharacterValid {
   type=NiagaraDataInterfaceArrayInt32
   description=""
   defaultValue {
    struct=NiagaraExt_VariableValue_DataInterface
    value {
     dataInterface=&self:NiagaraDataInterfaceArrayInt32_0
     dataInterfaceClass=&NiagaraDataInterfaceArrayInt32
    }
   }
  }
  User.LocalSpaceSpawnLocation {type=Vector3f, description="", defaultValue {struct=Vector3f, value=<0 0 0>}}
  User.MultiSphereCenter {type=Vector3f, description="", defaultValue {struct=Vector3f, value=<0 0 0>}}
  User.MultiSphereRightVector {type=Vector3f, description="", defaultValue {struct=Vector3f, value=<1000 0 0>}}
  User.MultiSphereUpVector {type=Vector3f, description="", defaultValue {struct=Vector3f, value=<0 0 1000>}}
  User.NewNiagaraDataInterfaceSkeletalMesh {
   type=NiagaraDataInterfaceSkeletalMesh
   defaultValue {
    struct=NiagaraExt_VariableValue_DataInterface
    value {dataInterface=&self:NiagaraDataInterfaceSkeletalMesh_0}
   }
  }
  User.Number Of Spheres {type=NiagaraInt32, defaultValue=5}
  User.PlayerVelocity {type=Vector3f, description="", defaultValue {struct=Vector3f, value=<0 0 0>}}
  User.PolyhedraCenter {type=Vector3f, description="", defaultValue {struct=Vector3f, value=<0 0 0>}}
  User.PolyhedraXAxis {type=Vector3f, description="", defaultValue {struct=Vector3f, value=<1 0 0>}}
  User.PolyhedraYAxis {type=Vector3f, description="", defaultValue {struct=Vector3f, value=<0 1 0>}}
  User.TargetArrangement {type=NiagaraInt32}
 ]
}
