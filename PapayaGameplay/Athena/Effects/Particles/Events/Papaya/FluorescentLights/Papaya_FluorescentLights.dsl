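; Papaya event actor that drives a Niagara swarm of fluorescent lights through switchable arrangements.
; Setup converts gizmo placements to world space and feeds exit, sphere, polyhedra and skeletal mesh data to Niagara.
; Custom events activate, set, cycle, randomize or kill patterns through an enum pushed as TargetArrangement.
; One pattern snapshots players nearest the local pawn, then ticks their positions into Niagara, dropping those far from the tornado until none remain.
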(event EventBeginPlay
  ;[0.4121 1 0.08369] Initialize
  (Utilities|FlowControl|Sequence
    (:then_0
      ;;[1 1 1] Transform gizmos to ws
      (Variables|PlayerSwarm|SetInternal%20World%20Space%20Screen%20Location (Math|Transform|TransformLocation (Transformation|GetActorTransform) (Variables|PlayerSwarm|GetScreen%20Location))))
    (:then_1
      (Variables|PlayerSwarm|SetTornado%20Location%20In%20World%20Space (Math|Transform|TransformLocation (Transformation|GetActorTransform) (Variables|PlayerSwarm|GetTornado%20Location))))
    (:then_2
      (Utilities|Array|Clear (Variables|ExitLocations|GetWorld%20Space%20Exit%20Positions))
      (for _array_element (Variables|ExitLocations|GetExit%20Positions)
        (Utilities|Array|Add (Variables|ExitLocations|GetWorld%20Space%20Exit%20Positions) (Math|Transform|TransformLocation (Transformation|GetActorTransform) _array_element))))
    (:then_3
      (CallFunction|SetMultiSphereProperties)
      (CallFunction|SetPolyhedraProperties)
      ;;[1 1 1] Initialize
      (Effects|Components|ParticleSystem|SetVectorParameter (Variables|Default|GetNiagara) "LocalSpaceSpawnLocation" (Variables|Default|GetLocal%20Space%20Spawn%20Placement))
      ;;;[1 1 1] Init phase
      (Niagara|SetNiagaraVariable(Int32) (Variables|Default|GetNiagara) "TargetArrangement" (Math|Conversions|ToInteger(Byte) (Utilities|Enum|LiteralenumENiagara_Papaya_FluorescentLightArrangements "NewEnumerator0")))
      ;;;[1 1 1] init skel
      (Utilities|IsValid (Variables|SkeletalMesh|GetTarget%20Skeletal%20Mesh)
        (:Is%20Valid
          (Niagara|SetNiagaraSkeletalMeshComponent (Variables|Default|GetNiagara) "NewNiagaraDataInterfaceSkeletalMesh" (Variables|SkeletalMesh|GetTarget%20Skeletal%20Mesh))
          (Niagara|NiagaraSetVectorArray (Variables|Default|GetNiagara) "ExitLocations" (Variables|ExitLocations|GetWorld%20Space%20Exit%20Positions))
          (Variables|Default|SetInitialized true))))))

(event Custom|ActivateLights
  ;[0.9979 1 0] Activate -
  (if (Variables|Default|GetInitialized)
    (Components|Activation|SetActive (Variables|Default|GetNiagara) true false)
    ;;[1 1 1] If the light is already active then increase the activation count so the particle system can internally handle this case
    (bind _output_get (Variables|Default|SetActive%20Light%20Pattern "NewEnumerator0"))
    (Niagara|SetNiagaraVariable(Int32) (Variables|Default|GetNiagara) "TargetArrangement" (Math|Conversions|ToInteger(Byte) _output_get))
    (bind _result (Math|Integer|IncrementInt (Variables|Default|GetActivation%20Count))
      (:%20%20
        (Effects|Components|ParticleSystem|SetIntParameter (Variables|Default|GetNiagara) "ActivationCount" _result) ; !!
      ))))

(event Custom|Set%20Pattern
  (params (Desired%20Behavior :type ENiagara_Papaya_FluorescentLightArrangements%20Enum :default "NewEnumerator0"))
  ;[1 0.1739 0.0542] Shift between light patterns
  ;;[1 1 1] Directly set the target pattern
  (Variables|Default|SetActive%20Light%20Pattern Desired%20Behavior)
  (CallFunction|ApplyShift true))

(event Custom|ApplyShift
  (params (Spawn%20If%20Deactivated :type Boolean :default true))
  (bind _active_light_pattern (Variables|Default|GetActive%20Light%20Pattern))
  (bind _returnvalue (Utilities|Enum|Equal(Enum) _active_light_pattern "NewEnumerator6"))
  ;[1 0.1739 0.0542] Shift between light patterns
  ;;[1 1 1] Apply new behaviors
  (if (Variables|Default|GetPreviousPatternWasDie)
    (if _returnvalue
      (Variables|Default|SetPreviousPatternWasDie true)
      (else
        _))
    (else
      (Variables|Default|SetPreviousPatternWasDie _returnvalue)
      (if (Components|Activation|IsActive (Variables|Default|GetNiagara))
        (CallFunction|ShiftPattern)
        (else
          _))))
  (if Spawn%20If%20Deactivated
    (Variables|Default|SetPreviousPatternWasDie false)
    (CallFunction|ActivateLights)))

(event Custom|KillLights
  ;[1 0.1739 0.0542] Shift between light patterns
  ;;[1 1 1] Kill Lights
  (Variables|Default|SetActive%20Light%20Pattern "NewEnumerator6")
  (CallFunction|ApplyShift false))

(event Custom|Shift%20Light%20Pattern
  ;[1 0.1739 0.0542] Shift between light patterns
  ;;[1 1 1] Increment behavior pattern
  (Variables|Default|SetActive%20Light%20Pattern (Utilities|Enum|BytetoEnumENiagara_Papaya_FluorescentLightArrangements (Math|Conversions|ToByte(Integer) (select (> (+ (Math|Conversions|ToInteger(Byte) (Variables|Default|GetActive%20Light%20Pattern)) 1) (- (Utilities|Enum|GetnumberofentriesinENiagara_Papaya_FluorescentLightArrangements) 1)) 0 (+ (Math|Conversions|ToInteger(Byte) (Variables|Default|GetActive%20Light%20Pattern)) 1)))))
  (CallFunction|ApplyShift true))

(event Custom|RandomizeLights
  ;[1 0.1739 0.0542] Shift between light patterns
  ;;[1 1 1] Randomize Light Pattern
  (Variables|Default|SetActive%20Light%20Pattern (Utilities|Enum|BytetoEnumENiagara_Papaya_FluorescentLightArrangements (Math|Conversions|ToByte(Integer) (Math|Random|RandomInteger (- (Utilities|Enum|GetnumberofentriesinENiagara_Papaya_FluorescentLightArrangements) 1)))))
  (CallFunction|ApplyShift true))

(event Custom|TrackPlayerPositions
  (params (Track :type Boolean))
  (bind _raw_overlapped_players (Variables|PlayerSwarm|GetRaw%20Overlapped%20Players))
  ;[0.8813 0.1721 1] Player tracking
  ;;[1 1 1] Initialize player tracking
  ;;;[1 1 1] Set the default behavior
  (Niagara|NiagaraSetInt32Array (Variables|Default|GetNiagara) "IsCharacterValid" (Utilities|Array|MakeArray 0))
  (Actor|Tick|SetActorTickEnabled Track)
  (if Track
    (Utilities|Array|Clear (Variables|PlayerSwarm|GetRaw%20Overlapped%20Players))
    (Utilities|Array|Clear (Variables|PlayerSwarm|GetClearedActorArray))
    (Utilities|IsValid (Game|GetPlayerPawn 0)
      (:Is%20Valid
        (Utilities|FlowControl|Sequence
          (:then_0
            ;;;[1 1 1] Find closest players
            (bind _outactors (Collision|SphereOverlapActors (Transformation|GetActorLocation (Game|GetPlayerPawn 0)) (Variables|PlayerSwarm|GetNear%20Player%20Target%20Selection%20Radius) (Utilities|Array|MakeArray "ObjectTypeQuery3") "/Game/Athena/PlayerPawn_Athena.PlayerPawn_Athena_C" (Utilities|Array|MakeArray (Game|GetPlayerPawn 0))))
            (Variables|PlayerSwarm|SetRaw%20Overlapped%20Players _outactors)
            ;;;[1 1 1] Store off the closest actor and remove it from the raw array
            (for _ (range (+ (- (Variables|PlayerSwarm|GetNumber%20of%20Players%20to%20track) 2) 1))
              (if (> (Utilities|Array|Length (Variables|PlayerSwarm|GetRaw%20Overlapped%20Players)) 0)
                (bind _returnvalue (Utilities|Array|Add (Variables|PlayerSwarm|GetClearedActorArray) (Fortnite|GetClosestActorfromArray (Game|GetPlayerPawn 0) _raw_overlapped_players)))
                (Utilities|Array|RemoveItem _raw_overlapped_players (Utilities|Array|Get(acopy) (Variables|PlayerSwarm|GetClearedActorArray) _returnvalue)))))
          (:then_1
            (Utilities|Array|Insert (Variables|PlayerSwarm|GetClearedActorArray) (Game|GetPlayerPawn 0))))))))

(event EventTick
  (params (DeltaSeconds :type Float%20%28single-precision%29))
  (bind _clearedactorarray (Variables|PlayerSwarm|GetClearedActorArray))
  ;[0.8813 0.1721 1] Player tracking
  ;;[1 1 1] Track Player Positions
  ;;;[1 1 1] remove any characters that leave the immediate area around the tornado
  (Utilities|Array|Clear (Variables|Tornado|GetRemoval%20Array))
  (bind (_arrayindex _arrayelement) (Utilities|Array|ReverseforEachLoop (Variables|PlayerSwarm|GetClearedActorArray))
    (:LoopBody
      (bind _distance (Class|MissionBlueprintFunctionLibrary|DistanceBetweenTwoVectors (Transformation|GetActorLocation _arrayelement) (Variables|PlayerSwarm|GetTornado%20Location%20In%20World%20Space)))
      (if (> _distance (Variables|PlayerSwarm|GetPlayer%20Distance%20to%20Tornado%20Before%20Rejoining%20the%20Group))
        (Utilities|Array|Add (Variables|Tornado|GetRemoval%20Array) _arrayindex)))
    (:Completed
      (for _array_element (Variables|Tornado|GetRemoval%20Array)
        (Utilities|Array|RemoveIndex (Variables|PlayerSwarm|GetClearedActorArray) _array_element))
      ;;;[1 1 1] If there aren't any players to track then zero out the arrays and stop ticking
      (if (== (Utilities|Array|Length (Variables|PlayerSwarm|GetClearedActorArray)) 0)
        (Utilities|Array|SetArrayElem (Variables|PlayerSwarm|GetIs%20valid%20character%20Array) (- (Variables|PlayerSwarm|GetNumber%20of%20Players%20to%20track) 1) 0 true) ; invalid locations
        (for _index (range (Variables|PlayerSwarm|GetNumber%20of%20Players%20to%20track))
          (Utilities|Array|SetArrayElem (Variables|PlayerSwarm|GetTrackedPositionArray) _index (Variables|PlayerSwarm|GetTornado%20Location%20In%20World%20Space) true) ; Invalid locations
        )
        (CallFunction|FeedPlayerTrackingVariablesIntoNiagara)
        (CallFunction|TrackPlayerPositions false)
        (else
          (Utilities|FlowControl|Sequence
            (:then_0
              ;;;[1 1 1] Fill arrays
              (Utilities|Array|Clear (Variables|PlayerSwarm|GetIs%20valid%20character%20Array))
              ;;;;[1 1 1] Init valid array
              (Utilities|Array|SetArrayElem (Variables|PlayerSwarm|GetIs%20valid%20character%20Array) (- (Variables|PlayerSwarm|GetNumber%20of%20Players%20to%20track) 1) 0 true) ; invalid locations
              (for _index_2 (range (Variables|PlayerSwarm|GetNumber%20of%20Players%20to%20track))
                (if (Utilities|Array|IsValidIndex (Variables|PlayerSwarm|GetClearedActorArray) _index_2)
                  (if (Utilities|IsValid (Utilities|Array|Get(acopy) _clearedactorarray _index_2) :!class /Script/Engine.KismetSystemLibrary)
                    ;;;;[1 1 1] Track valid characters
                    (Utilities|Array|SetArrayElem (Variables|PlayerSwarm|GetIs%20valid%20character%20Array) _index_2 1 true) ; valid locations
                    (Utilities|Array|SetArrayElem (Variables|PlayerSwarm|GetTrackedPositionArray) _index_2 (Transformation|GetActorLocation (Utilities|Array|Get(acopy) _clearedactorarray _index_2)) true) ; valid locations
                    (else
                      _))
                  (else
                    _))
                (Utilities|Array|SetArrayElem (Variables|PlayerSwarm|GetTrackedPositionArray) _index_2 (Variables|PlayerSwarm|GetTornado%20Location%20In%20World%20Space) true) ; Invalid locations
              ))
            (:then_1
              (CallFunction|FeedPlayerTrackingVariablesIntoNiagara))))))))

(fn AllowShiftAgain
  (Variables|Default|SetCanShiftAgain true)
  (if (Variables|Default|GetDelayedShift)
    (Variables|Default|SetDelayedShift false)
    (CallFunction|ShiftPattern)))

(fn FeedPlayerTrackingVariablesintoNiagara
  (Niagara|NiagaraSetVectorArray (Variables|Default|GetNiagara) "ArrayOfNearestPlayerPositions" (Variables|PlayerSwarm|GetTrackedPositionArray))
  (Niagara|NiagaraSetInt32Array (Variables|Default|GetNiagara) "IsCharacterValid" (Variables|PlayerSwarm|GetIs%20valid%20character%20Array))
  (Effects|Components|ParticleSystem|SetVectorParameter (Variables|Default|GetNiagara) "PlayerVelocity" (select (Utilities|IsValid (Game|GetPlayerPawn 0) :!class /Script/Engine.KismetSystemLibrary) (Transformation|GetVelocity (Game|GetPlayerPawn 0)) "0, 0, 0")))

(fn SetMultiSphereProperties
  (bind _returnvalue (Transformation|GetActorTransform))
  (bind _multi_sphere_behaviour_placement (Variables|MultiSphere|GetMulti%20Sphere%20Behaviour%20Placement))
  (bind _returnvalue_2 (Math|Transform|ComposeTransforms _multi_sphere_behaviour_placement _returnvalue))
  (bind _rotation (.rotation _returnvalue_2))
  (bind _scale (.scale _returnvalue_2))
  (bind (_x _y _z) (Math|Rotator|GetAxes _rotation))
  (bind _returnvalue_3 1e+03)
  ;[1 1 1] Polyhedra and Sphere Arrangement
  (Effects|Components|ParticleSystem|SetVectorParameter (Variables|Default|GetNiagara) (Utilities|Name|MakeLiteralName "MultiSphereCenter") (.location _returnvalue_2))
  (Effects|Components|ParticleSystem|SetVectorParameter (Variables|Default|GetNiagara) (Utilities|Name|MakeLiteralName "MultiSphereUpVector") (* _y (* (.y _scale) _returnvalue_3)))
  (Effects|Components|ParticleSystem|SetVectorParameter (Variables|Default|GetNiagara) (Utilities|Name|MakeLiteralName "MultiSphereRightVector") (* _z (* (.z _scale) _returnvalue_3))))

(fn SetPolyhedraProperties
  (bind _returnvalue (Transformation|GetActorTransform))
  (bind _polyhedratransform (Variables|Polyhedra|GetPolyhedraTransform))
  (bind _returnvalue_2 (Math|Transform|ComposeTransforms _polyhedratransform _returnvalue))
  (bind _scale (.scale _returnvalue_2))
  (bind _returnvalue_3 (* _scale 1e+03))
  (bind (_x _y) (Math|Rotator|GetAxes (.rotation _returnvalue_2)))
  (bind _rotation (.rotation _returnvalue_2))
  (Effects|Components|ParticleSystem|SetVectorParameter (Variables|Default|GetNiagara) "PolyhedraCenter" (.location _returnvalue_2))
  (Effects|Components|ParticleSystem|SetVectorParameter (Variables|Default|GetNiagara) "PolyhedraXAxis" (* _x (.x _returnvalue_3)))
  (Effects|Components|ParticleSystem|SetVectorParameter (Variables|Default|GetNiagara) "PolyhedraYAxis" (* _y (.y _returnvalue_3))))

(fn ShiftPattern
  (Niagara|SetNiagaraVariable(Int32) (Variables|Default|GetNiagara) "TargetArrangement" (Math|Conversions|ToInteger(Byte) (Variables|Default|GetActive%20Light%20Pattern)))
  (CallFunction|TrackPlayerPositions (Utilities|Enum|Equal(Enum) (Variables|Default|GetActive%20Light%20Pattern) "NewEnumerator3")))

(fn ConstructionScript
  (bind _target_skeletal_mesh (Variables|SkeletalMesh|GetTarget%20Skeletal%20Mesh))
  (Utilities|IsValid _target_skeletal_mesh
    (:Is%20Valid
      (Niagara|SetNiagaraSkeletalMeshComponent (Variables|Default|GetNiagara) "NewNiagaraDataInterfaceSkeletalMesh" _target_skeletal_mesh)
      (Effects|Components|ParticleSystem|SetVectorParameter (Variables|Default|GetNiagara) "LocalSpaceSpawnLocation" (Variables|Default|GetLocal%20Space%20Spawn%20Placement))
      (Transformation|SetRelativeTransform (Variables|SkeletalMesh|GetTarget%20Skeletal%20Mesh) (Variables|SkeletalMesh|GetSkeletal%20Mesh%20Relative%20Transform) false true)
      (CallFunction|SetMultiSphereProperties)
      (CallFunction|SetPolyhedraProperties))))
